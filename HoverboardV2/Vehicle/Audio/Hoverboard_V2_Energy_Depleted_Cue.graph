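# Sound cue for the Hoverboard V2 energy-depleted sound.
# A mixer layers the depleted sound at full volume with an alarm wave at reduced volume, delayed briefly by a fixed amount.
# It uses the Fortnite vehicle prop attenuation and interior vehicle sound class, and overrides concurrency to prevent new instances when the limit is reached.
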SoundCue /HoverboardV2/Vehicle/Audio/Hoverboard_V2_Energy_Depleted_Cue
root=Mixer_0
nodes [
 {id=Delay_0, children [WavePlayer_0], properties {delayMax=0.2, delayMin=0.2}}
 {id=Mixer_0, children [WavePlayer_2,Delay_0], properties {inputVolume [1,0.6]}}
 {id=WavePlayer_0, properties {soundWaveAssetPtr=&/Hoverboard_V2_Energy_Depleted_Alarrm}}
 {id=WavePlayer_2, properties {soundWaveAssetPtr=&/Hoverboard_V2_Energy_Depleted}}
]
properties {
 attenuationSettings=/Game/Sounds/Attenuation/Fort_Prop_Vehicles_Att
 bOverrideConcurrency=true
 maxDistance=3000
 soundClassObject=&Fort_Vehicles_Interior
 volumeMultiplier=1
 concurrencyOverrides {resolutionRule=PreventNew}
}
