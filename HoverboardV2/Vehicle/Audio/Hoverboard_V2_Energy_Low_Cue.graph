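# Sound cue that plays the hoverboard's low-energy start sound, Hoverboard_V2_Energy_Low_Start, as part of the HoverboardV2 vehicle audio.
# A single wave player feeds the cue output directly, with no looping or randomization nodes.
# It uses the shared vehicle prop attenuation and interior vehicle sound class at reduced volume.
# Overridden concurrency blocks new instances instead of stopping sounds already playing.
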SoundCue /HoverboardV2/Vehicle/Audio/Hoverboard_V2_Energy_Low_Cue
root=WavePlayer_2
nodes [{id=WavePlayer_2, properties {soundWaveAssetPtr=&/Hoverboard_V2_Energy_Low_Start}}]
properties {
 attenuationSettings=/Game/Sounds/Attenuation/Fort_Prop_Vehicles_Att
 bOverrideConcurrency=true
 maxDistance=3000
 soundClassObject=&Fort_Vehicles_Interior
 volumeMultiplier=0.7
 concurrencyOverrides {resolutionRule=PreventNew}
}
