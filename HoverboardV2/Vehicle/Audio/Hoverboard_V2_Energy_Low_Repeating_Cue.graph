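# Sound cue for the low-energy sound of the V2 Hoverboard vehicle, named as repeating though it has no looping node.
# A Random node picks with equal weight between the main low-energy wave and an alternate wave, which passes through a single-input Mixer that boosts its volume.
# The cue plays at reduced volume with vehicle prop attenuation and the vehicle interior sound class.
# Overridden concurrency blocks new instances at the limit.
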SoundCue /HoverboardV2/Vehicle/Audio/Hoverboard_V2_Energy_Low_Repeating_Cue
root=Random_0
nodes [
 {id=Mixer_0, children [WavePlayer_0], properties {inputVolume [3]}}
 {id=Random_0, children [WavePlayer_2,Mixer_0], properties {bShouldExcludeFromBranchCulling=true, weights [1,1]}}
 {id=WavePlayer_0, properties {soundWaveAssetPtr=&/Hoverboard_V2_Energy_Low_Alt}}
 {id=WavePlayer_2, properties {soundWaveAssetPtr=&/Hoverboard_V2_Energy_Low}}
]
properties {
 attenuationSettings=/Game/Sounds/Attenuation/Fort_Prop_Vehicles_Att
 bOverrideConcurrency=true
 maxDistance=3000
 soundClassObject=&Fort_Vehicles_Interior
 volumeMultiplier=0.5
 concurrencyOverrides {resolutionRule=PreventNew}
}
