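# Looping sound cue for the LinedNotebookVehicle in the HoverboardV2 content, named for the vehicle's in-water state.
# A looping water wave feeds an Enveloper that ramps volume up from silence over the first half second.
# A root Modulator at default settings sits on top.
# The cue plays in the Fortnite vehicles sound class at reduced volume, using the vehicle's close movement attenuation.
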SoundCue /HoverboardV2/Vehicle/Audio/IP_LinedNotebookVehicle_InWater_Loop_Cue
root=Modulator_0
nodes [
 {
  id=Enveloper_0
  children [WavePlayer_0]
  properties {pitchCurve {editorCurveData {keys [{time=0}]}}, volumeCurve {editorCurveData {keys [<at 0 0>,<at 0.5 1>]}}}
 }
 {id=Modulator_0, children [Enveloper_0]}
 {id=WavePlayer_0, properties {bLooping=true, soundWaveAssetPtr=&/LinedNotebookVehicle_Water_loop}}
]
properties {
 attenuationSettings=/HoverboardV2/Vehicle/Audio/Settings/LinedNotebookVehicle_Movement_Close_Att
 maxDistance=5300
 soundClassObject=&Fort_Vehicles
 volumeMultiplier=0.6
}
