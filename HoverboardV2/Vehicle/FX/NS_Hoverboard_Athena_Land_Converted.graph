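# Niagara burst effect for a hoverboard landing, converted from Cascade, with smoke, a ground ripple, a flickering light and lightning meshes.
# Lightning spawns on an expanding cylinder with flickering alpha; its color, plus the light's and ripple's, switches between cyan and green via System.ColorSet.
# User.Water suppresses the continuous smoke spawn, not its burst, and User.Moving feeds the ripple material.
# Water-driven foam and slow-ripple emitters are present but disabled.
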NiagaraSystem /HoverboardV2/Vehicle/FX/NS_Hoverboard_Athena_Land_Converted
dependencies {
 usedDataInterfaces [
  &NiagaraDataInterfaceCurve
  &NiagaraDataInterfaceVectorCurve
  &NiagaraDataInterfaceSpriteRendererInfo
  &NiagaraDataInterfaceArrayColor
 ]
 usedDynamicInputs []
 usedModules [
  &SystemState
  &self:SystemSpawnScript.SystemScriptSource.SystemScriptGraph.NiagaraNodeAssignment_0.SetVariables_#1
 ]
}
emitters {
 FoamRollingOut {
  bEnabled=false
  emitterUpdateScript [
   &EmitterState(Life Cycle Mode=::NewEnumerator1, Loop Behavior=::NewEnumerator1, Loop Duration=1, UseLoopDelay=true)
   &SpawnBurst_Instantaneous(Spawn Count=2, Spawn Probability=$User.Water, Use Spawn Probability=true, UseLoopCountLimit=false)
  ]
  particleSpawnScript [
   &InitializeParticle(
    Lifetime=&RandomRangeFloat(Minimum=1, Maximum=1.4)
    Mesh Scale=&RandomRangeVector(Minimum=<1 1 4>, Maximum=<1.4 1.4 5>)
    Mesh Scale Mode=::NewEnumerator4
    UsePositionOffset=false
    Write Material Random=true
   )
   &Color(Color=&MakeLinearColorFromVectorAndFloat(
     "Vector (RGB)"=<0.75 0.75 0.75>
     "Float (Alpha)"=&ClampFloat(Float=&RandomRangeFloat(Minimum=0.3, Maximum=0.5))
    ))
   &InheritVelocity(Inherited Velocity Amount Scale=<1 1 0>, Inherited Velocity Speed Limit=0, Speed Limit=false, UseSpeedThreshold=false)
   &AddVelocity(Velocity=&RandomRangeVector(Minimum=<-25 -25 0>, Maximum=<25 25 0>))
   &InitialMeshOrientation(
    Mesh Orientation Mode=::NewEnumerator7
    Rotation=&RandomRangeVector(Minimum=<0 0 0>, Maximum=<0 0 1>)
    Use Rotation Vector=true
   )
  ]
  particleUpdateScript [
   &ParticleState()
   &ScaleMeshSize(Scale Factor=&VectorFromCurve(VectorCurve=DataInterface() {propertyValues {
       XCurve {keys [<at 0 0 RCIM_Cubic>, {arriveTangent=2.4, interpMode=RCIM_Cubic, leaveTangent=2.4, time=0.25, value=1.25}, <at 1 2.4 RCIM_Cubic>]}
       YCurve {keys [<at 0 0 RCIM_Cubic>, {arriveTangent=2.4, interpMode=RCIM_Cubic, leaveTangent=2.4, time=0.25, value=1.25}, <at 1 2.4 RCIM_Cubic>]}
       ZCurve {keys [
         {arriveTangent=42.4744, interpMode=RCIM_Cubic, leaveTangent=42.4744, tangentMode=RCTM_User, time=0, value=0.25}
         {arriveTangent=1.5, interpMode=RCIM_Cubic, leaveTangent=1.5, time=0.08, value=2.24}
         {arriveTangent=-7.7308, interpMode=RCIM_Cubic, leaveTangent=13.5525, tangentMode=RCTM_Break, time=0.16, value=0.49}
         {arriveTangent=-0.4643, interpMode=RCIM_Cubic, leaveTangent=-0.4643, time=0.34, value=0.23}
         <at 1 0.1 RCIM_Cubic>
        ]}
      }}))
   &ScaleColor(
    Color Value To Scale=$Particles.Color
    Scale Alpha=&FloatFromCurve(FloatCurve=DataInterface() {propertyValues {Curve {keys [<at 0 0 RCIM_Cubic>,<at 0.25 1 RCIM_Cubic>,<at 1 0>]}}})
    Scale RGB=&VectorFromCurve(VectorCurve=DataInterface() {propertyValues {XCurve {keys [<at 0 0.5793>,<at 1 0.2095>]}, YCurve {keys [<at 0 0.7235>,<at 1 0.2746>]}, ZCurve {keys [<at 0 0.77>,<at 1 0.3021>]}}})
    ScaleA=true
    ScaleRGB=true
   )
   &Drag(Drag=4, Use Linear Drag=true, Use Rotational Drag=true)
   &SolveForcesAndVelocity(Clamp Velocity=false, Limit Acceleration=false)
   &CameraOffset(Camera Offset Amount=16)
   &DynamicMaterialParameters(
    Index 0 Param 1=&FloatFromCurve(FloatCurve=DataInterface() {propertyValues {Curve {keys [
         <at 0.0022 0.6 RCIM_Cubic>
         {arriveTangent=-1.6071, interpMode=RCIM_Cubic, leaveTangent=-1.6071, time=0.25, value=1}
         <at 0.9979 -1 RCIM_Cubic>
        ]}}})
    Index 0 Param 2=&FloatFromCurve(FloatCurve=DataInterface() {propertyValues {Curve {keys [<at 0.0043 1 RCIM_Cubic>,<at 1 0 RCIM_Cubic>]}}})
    Index 0 Param 3=&RandomRangeFloat(Minimum=2, Maximum=3)
    Module.Index0Param0Write=true
    Module.Index0Param1Write=true
    Module.Index0Param2Write=true
    Module.Index0Param3Write=true
   ) {module=DynamicMaterialParameters002}
   &ScaleColor(
    Scale Alpha=&FloatFromCurve(FloatCurve=DataInterface() {propertyValues {Curve {keys [<at 0 0 RCIM_Cubic>,<at 0.25 1 RCIM_Cubic>,<at 1 0>]}}})
    Scale RGB=&VectorFromCurve(VectorCurve=DataInterface() {propertyValues {XCurve {keys [<at 0 0.5793>,<at 1 0.2095>]}, YCurve {keys [<at 0 0.7235>,<at 1 0.2746>]}, ZCurve {keys [<at 0 0.77>,<at 1 0.3021>]}}})
    ScaleA=true
    ScaleRGB=true
   ) {module=ScaleColor001}
  ]
  rendererClasses [&NiagaraMeshRendererProperties]
  renderers [{rendererClass=&NiagaraMeshRendererProperties, rendererIndex=0}]
 }
 Light {
  emitterUpdateScript [
   &EmitterState(Life Cycle Mode=::NewEnumerator1, Loop Behavior=::NewEnumerator1, Loop Duration=0.5, UseLoopDelay=true)
   &SpawnBurst_Instantaneous(Use Spawn Probability=false, UseLoopCountLimit=false)
  ]
  particleSpawnScript [&InitializeParticle(
    Color=&SelectLinearColorFromArray(
     Array Sampling Mode=::NewEnumerator1
     Direct Array Index=$System.ColorSet
     Color Selection Array=DataInterface() {propertyValues {ColorData [<rgba 0.1832 1.1956 2 1>, <rgba 0.1832 2 0.1832 1>]}}
     Random Seed=?
    )
    Lifetime=&RandomRangeFloat(Minimum=0.25, Maximum=0.3)
    Position=&ConvertVectorToPosition(Input Position=<0 0 30>, Conversion Operation:ENiagara_LWCConvertVecToPos=::NewEnumerator0)
    Position Mode=::NewEnumerator1
    Sprite Size=<300 300>
    Sprite Size Mode=::NewEnumerator4
    UsePositionOffset=false
    Write Material Random=true
   )]
  particleUpdateScript [
   /SetVariables(Particles.LightExponent=8) {isSetParameters=true}
   &ParticleState()
   &CascadeConversion_LightProperties(LightBrightnessScale=32, LightColorScale=<1 1 1>, LightRadiusScale=1)
   &ScaleColor(
    Color Value To Scale=$Particles.Color
    Scale Alpha=&FloatFromCurve(FloatCurve=DataInterface() {propertyValues {Curve {keys [
         <at 0 0 RCIM_Cubic>
         {arriveTangent=10, interpMode=RCIM_Cubic, leaveTangent=10, time=0.0395, value=2}
         {arriveTangent=-3.6226, interpMode=RCIM_Cubic, leaveTangent=-3.6226, time=0.1, value=1}
         {arriveTangent=-0.4208, interpMode=RCIM_Cubic, leaveTangent=-0.4208, time=0.5024, value=0.3229}
         {arriveTangent=-4.1501, interpMode=RCIM_Cubic, leaveTangent=-4.1501, time=0.5428, value=0.8137}
         {arriveTangent=-6.8779, interpMode=RCIM_Cubic, leaveTangent=-6.8779, time=0.557, value=0.0964}
         {arriveTangent=4.0651, interpMode=RCIM_Cubic, leaveTangent=4.0651, time=0.6479, value=0.0908}
         {arriveTangent=0.2604, interpMode=RCIM_Cubic, leaveTangent=0.2604, time=0.6823, value=0.6058}
         {arriveTangent=-1.8836, interpMode=RCIM_Cubic, leaveTangent=-1.8836, time=0.7409, value=0.115}
         {arriveTangent=-0.444, interpMode=RCIM_Cubic, leaveTangent=-0.444, time=0.844, value=0.3013}
         <at 1 0>
        ]}}})
    Scale RGB=&VectorFromCurve(VectorCurve=DataInterface() {propertyValues {XCurve {keys [<at 0 1>,<at 1 1>]}, YCurve {keys [<at 0 1>,<at 1 1>]}, ZCurve {keys [<at 0 1>,<at 1 1>]}}})
    ScaleA=true
    ScaleRGB=true
   )
  ]
  rendererClasses [&NiagaraLightRendererProperties]
  renderers [{rendererClass=&NiagaraLightRendererProperties, rendererIndex=0}]
 }
 Lightning {
  emitterUpdateScript [
   &EmitterState(Life Cycle Mode=::NewEnumerator1, Loop Behavior=::NewEnumerator1, Loop Duration=0.5, UseLoopDelay=true)
   &SpawnBurst_Instantaneous(
    Spawn Count=&RandomRangeInt(Minimum=6, Maximum=3)
    Spawn Time=0.01
    Use Spawn Probability=false
    UseLoopCountLimit=false
   ) {module=SpawnBurst_Instantaneous001}
   &SpawnBurst_Instantaneous(
    Spawn Count=&RandomRangeInt(Minimum=3, Maximum=1)
    Spawn Time=0.08
    Use Spawn Probability=false
    UseLoopCountLimit=false
   )
   &SpawnRate(
    SpawnRate=&Multiply_Float(
     A=25
     B=&FloatFromCurve(FloatCurve=DataInterface() {propertyValues {Curve {keys [<at 0 1 RCIM_Cubic>,<at 0.35 0>]}}}, CurveIndex=$Emitter.NormalizedLoopAge)
    )
    Use Spawn Probability=false
   )
  ]
  particleSpawnScript [
   &InitializeParticle(
    Color=&SelectLinearColorFromArray(
     Array Sampling Mode=::NewEnumerator1
     Direct Array Index=$System.ColorSet
     Color Selection Array=DataInterface() {propertyValues {ColorData [<rgba 0.4407 1.4076 2 1>, <rgba 0.4167 2 0.1 1>]}}
     Random Seed=?
    )
    Lifetime=&RandomRangeFloat(Minimum=0.12, Maximum=0.25)
    Mesh Scale=&RandomRangeVector(Minimum=<1 1 2>, Maximum=<2 2 4>)
    Mesh Scale Mode=::NewEnumerator4
    Position Offset=<0 0 -5>
    UsePositionOffset=true
    Write Material Random=true
   )
   &AddVelocity(Velocity=&RandomRangeVector(Minimum=<10 -20 0>, Maximum=<30 20 0>))
   &ShapeLocation(
    Cylinder Height=5
    Cylinder Radius=&FloatFromCurve(
     FloatCurve=DataInterface() {propertyValues {Curve {keys [{arriveTangent=683.0251, interpMode=RCIM_Cubic, leaveTangent=683.0251, tangentMode=RCTM_User, time=0, value=10}, <at 0.25 100>]}}}
     CurveIndex=$Emitter.LoopedAge
    )
    MaskSpawnGroup=false
    Offset=<0 0 5>
    Offset Mode=::NewEnumerator0
    Override Seed=false
    Rotation Axis=<0 0 1>
    Rotation Mode=::NewEnumerator1
    Shape Primitive=::NewEnumerator1
    Surface Only=true
    Use Endcaps In Surface Only Mode=false
    UseOwnerScale=true
   )
   &AddVelocity(
    Velocity=&Multiply_VectorByFloat(
     Vector=&Multiply_Vector(A=$Particles.ShapeLocation.ShapeVector, B=<1 1 0>)
     Float=&RandomRangeFloat(Minimum=2, Maximum=5)
    )
    Velocity Speed Scale=5
   ) {module=AddVelocity001}
   &InitialMeshOrientation(
    Mesh Orientation Mode=::NewEnumerator7
    Rotation=&RandomRangeVector(Minimum=<-0.1 0 0>, Maximum=<0.1 0.1 1>)
    Use Rotation Vector=true
   )
   &CameraOffset()
  ]
  particleUpdateScript [
   &ParticleState()
   &ScaleMeshSize(Scale Factor=&VectorFromCurve(VectorCurve=DataInterface() {propertyValues {
       XCurve {keys [
         <at 0 1 RCIM_Cubic>
         {arriveTangent=1.9237, interpMode=RCIM_Cubic, leaveTangent=1.9237, tangentMode=RCTM_User, time=0.3573, value=1.2918}
         <at 1 1>
        ]}
       YCurve {keys [<at 0 2 RCIM_Cubic>, {arriveTangent=-1, interpMode=RCIM_Cubic, leaveTangent=-1, time=0.3573, value=1.7082}, <at 1 1>]}
       ZCurve {keys [<at 0 1 RCIM_Cubic>, {arriveTangent=-0.9, interpMode=RCIM_Cubic, leaveTangent=-0.9, time=0.3573, value=2}, <at 1 0.1>]}
      }}))
   &UpdateMeshOrientation(Rotation Vector=&RandomRangeVector(Minimum=<0 0 0>, Maximum=<0 0 1>))
   &Drag(Use Linear Drag=true, Use Rotational Drag=true)
   &SolveForcesAndVelocity(Clamp Velocity=false, Limit Acceleration=false)
   &ScaleColor(
    Scale Alpha=&FloatFromCurve(FloatCurve=DataInterface() {propertyValues {Curve {keys [
         <at 0.4 1 RCIM_Cubic>
         {arriveTangent=-6.9248, interpMode=RCIM_Cubic, leaveTangent=-6.9248, time=0.4927, value=1.1208}
         {arriveTangent=0.3732, interpMode=RCIM_Cubic, leaveTangent=0.3732, time=0.5399, value=0.0315}
         {arriveTangent=-0.022, interpMode=RCIM_Cubic, leaveTangent=-0.022, time=0.5887, value=1.1566}
         {arriveTangent=0.0713, interpMode=RCIM_Cubic, leaveTangent=0.0713, time=0.6342, value=0.0294}
         {arriveTangent=-0.1642, interpMode=RCIM_Cubic, leaveTangent=-0.1642, time=0.6679, value=1.1623}
         {arriveTangent=-4.6994, interpMode=RCIM_Cubic, leaveTangent=-4.6994, time=0.7083, value=0.0172}
         {arriveTangent=-0.0103, interpMode=RCIM_Cubic, leaveTangent=-0.0103, time=0.7605, value=0.727}
         {arriveTangent=-0.453, interpMode=RCIM_Cubic, leaveTangent=-0.453, time=0.8161, value=0.0161}
         {arriveTangent=0.0387, interpMode=RCIM_Cubic, leaveTangent=0.0387, time=0.8565, value=0.6835}
         {arriveTangent=-2.9465, interpMode=RCIM_Cubic, leaveTangent=-2.9465, time=0.9045, value=0.0196}
         {arriveTangent=-0.2048, interpMode=RCIM_Cubic, leaveTangent=-0.2048, time=0.9441, value=0.4254}
         <at 1 0>
        ]}}})
    ScaleA=true
    ScaleRGB=true
   ) {module=ScaleColor001}
   &DynamicMaterialParameters(
    Index 0 Param 1=&FloatFromCurve(FloatCurve=DataInterface() {propertyValues {Curve {keys [
         {arriveTangent=45.9263, interpMode=RCIM_Cubic, time=0, value=0}
         {arriveTangent=3.3333, interpMode=RCIM_Cubic, leaveTangent=3.3333, time=0.1, value=1.5}
         <at 0.3 1>
         {arriveTangent=-1.4286, interpMode=RCIM_Cubic, leaveTangent=-1.4286, time=0.7, value=1}
         <at 1 0>
        ]}}})
    Index 0 Param 2=&FloatFromCurve(FloatCurve=DataInterface() {propertyValues {Curve {keys [<at 0 0.4 RCIM_Cubic>, {arriveTangent=-0.4, interpMode=RCIM_Cubic, leaveTangent=-0.4, time=0.8, value=2}, <at 1 0>]}}})
    Index 0 Param 3=&RandomRangeFloat(Minimum=0, Maximum=1)
    Index 0 Param 4=&RandomRangeFloat(Minimum=0, Maximum=1)
    Module.Index0Param0Write=true
    Module.Index0Param1Write=true
    Module.Index0Param2Write=true
    Module.Index0Param3Write=true
   ) {module=DynamicMaterialParameters002}
  ]
  rendererClasses [&NiagaraMeshRendererProperties]
  renderers [{rendererClass=&NiagaraMeshRendererProperties, rendererIndex=0}]
 }
 Ripple {
  emitterUpdateScript [
   &EmitterState(Life Cycle Mode=::NewEnumerator1, Loop Behavior=::NewEnumerator1, Loop Duration=0.25, UseLoopDelay=true)
   &SpawnBurst_Instantaneous(Use Spawn Probability=false, UseLoopCountLimit=false)
  ]
  particleSpawnScript [
   /SetVariables(Particles.SpriteFacing=<0 0 1>) {isSetParameters=true}
   &InitializeParticle(
    Color=&SelectLinearColorFromArray(
     Array Sampling Mode=::NewEnumerator1
     Direct Array Index=$System.ColorSet
     Color Selection Array=DataInterface() {propertyValues {ColorData [<rgba 1.457 2.5169 4 1>, <rgba 1.457 4 1.457 1>]}}
     Random Seed=?
    )
    Lifetime=&RandomRangeFloat(Minimum=0.25, Maximum=0.3)
    Sprite Rotation Angle=0.25
    Sprite Rotation Mode=::NewEnumerator1
    Sprite Size=<200 200>
    Sprite Size Mode=::NewEnumerator4
    UsePositionOffset=false
    Write Material Random=true
   )
  ]
  particleUpdateScript [
   &ParticleState()
   &ScaleSpriteSize(
    Scale Sprite Size Mode=::NewEnumerator0
    Uniform Scale Factor=&FloatFromCurve(FloatCurve=DataInterface() {propertyValues {Curve {keys [{arriveTangent=1.4655, interpMode=RCIM_Cubic, time=0, value=0.4}, <at 1 1>]}}})
   )
   &CameraOffset(Camera Offset Amount=45)
   &ScaleColor(
    Scale Alpha=&FloatFromCurve(FloatCurve=DataInterface() {propertyValues {Curve {keys [<at 0 0 RCIM_Cubic>,<at 0.1 1 RCIM_Cubic>,<at 1 0>]}}})
    ScaleA=true
    ScaleRGB=true
   ) {module=ScaleColor001}
   &DynamicMaterialParameters(
    Index 0 Param 1=&RandomRangeFloat(Minimum=0, Maximum=1)
    Index 0 Param 2=$User.Moving
    Index 0 Param 3=1
    Index 0 Param 4=1
    Module.Index0Param0Write=true
    Module.Index0Param1Write=true
    Module.Index0Param2Write=true
    Module.Index0Param3Write=true
   ) {module=DynamicMaterialParameters002}
  ]
  rendererClasses [&NiagaraSpriteRendererProperties]
  renderers [{rendererClass=&NiagaraSpriteRendererProperties}]
 }
 RipplesSlow {
  bEnabled=false
  emitterUpdateScript [
   &EmitterState(Life Cycle Mode=::NewEnumerator1, Loop Behavior=::NewEnumerator1, Loop Duration=0.2, UseLoopDelay=true)
   &SpawnBurst_Instantaneous(Spawn Count=5, Spawn Probability=$User.Water, Use Spawn Probability=true, UseLoopCountLimit=false)
   &SpawnRate(SpawnRate=&Multiply_Float(A=12, B=$User.Water), Use Spawn Probability=false)
  ]
  particleSpawnScript [
   /SetVariables(Particles.SpriteFacing=<0 0 1>) {isSetParameters=true}
   /SetVariables(Particles.SpriteRotationRate=&RandomRangeFloat(Minimum=-0.2, Maximum=0.2)) {isSetParameters=true}
   &InitializeParticle(
    Color=<rgba 0.5 0.5 0.5 1>
    Lifetime=&RandomRangeFloat(Minimum=0.5, Maximum=1)
    Position Offset=&RandomRangeVector(Minimum=<-20 0 5>, Maximum=<30 0 5>)
    Sprite Rotation Angle=&RandomRangeFloat(Minimum=0, Maximum=1)
    Sprite Rotation Mode=::NewEnumerator1
    Sprite Size=&Vector2DFromFloat(Value=&RandomRangeFloat(Minimum=200, Maximum=400))
    Sprite Size Mode=::NewEnumerator4
    UsePositionOffset=true
    Write Material Random=true
   )
   &InheritVelocity(Inherited Velocity Amount Scale=<1 1 0>, Inherited Velocity Speed Limit=0, Speed Limit=false, UseSpeedThreshold=false)
  ]
  particleUpdateScript [
   &ParticleState()
   &SpriteRotationRate(Rotation Rate=&AngleConversion(
     Angle:NiagaraFloat=$Particles.SpriteRotationRate
     Angle Input:ENiagara_AngleInput=::NewEnumerator1
     Angle Output:ENiagara_AngleInput=::NewEnumerator0
    ))
   &ScaleSpriteSize(
    Scale Sprite Size Mode=::NewEnumerator0
    Uniform Scale Factor=&FloatFromCurve(FloatCurve=DataInterface() {propertyValues {Curve {keys [<at 0 0 RCIM_Cubic>, {interpMode=RCIM_Cubic, leaveTangent=0.35, time=1, value=1}]}}})
   )
   &ScaleColor(
    Scale Alpha=&FloatFromCurve(FloatCurve=DataInterface() {propertyValues {Curve {keys [<at 0 0 RCIM_Cubic>,<at 1 1 RCIM_Cubic>]}}})
    ScaleA=true
    ScaleRGB=true
   ) {module=ScaleColor001}
   &CameraOffset(Camera Offset Amount=16)
   &Drag(Drag=2, Use Linear Drag=true, Use Rotational Drag=true)
   &SolveForcesAndVelocity(Clamp Velocity=false, Limit Acceleration=false)
  ]
  rendererClasses [&NiagaraSpriteRendererProperties]
  renderers [{rendererClass=&NiagaraSpriteRendererProperties}]
 }
 TransSmoke {
  emitterUpdateScript [
   &EmitterState(Life Cycle Mode=::NewEnumerator1, Loop Behavior=::NewEnumerator1, Loop Duration=0.1, UseLoopDelay=true)
   &SpawnBurst_Instantaneous(Spawn Count=8, Use Spawn Probability=false, UseLoopCountLimit=false)
   &SpawnRate(
    SpawnRate=&Multiply_Float(A=90, B=&OneMinusFloat(Float=$User.Water))
    Use Spawn Probability=false
   )
  ]
  particleSpawnScript [
   /SetVariables(Particles.SpriteRotationRate=&RandomRangeFloat(Minimum=-0.1, Maximum=0.1)) {isSetParameters=true}
   &InitializeParticle(
    Color Maximum=<rgba 1 1 1 0.025>
    Color Minimum=<rgba 1 1 1 0.05>
    Color Mode=::NewEnumerator2
    Lifetime=&RandomRangeFloat(Minimum=0.3, Maximum=0.5)
    Sprite Size=&Vector2DFromFloat(Value=&RandomRangeFloat(Minimum=60, Maximum=100))
    Sprite Size Mode=::NewEnumerator4
    UsePositionOffset=false
    Write Material Random=true
   )
   &Color(Color=&MakeLinearColorFromVectorAndFloat(
     "Vector (RGB)"=<1 1 1>
     "Float (Alpha)"=&ClampFloat(Float=&RandomRangeFloat(Minimum=0.1, Maximum=0.2))
    )) {enabled=false}
   &ShapeLocation(
    Cylinder Height=10
    Cylinder Radius=&RandomRangeFloat(Minimum=15, Maximum=22)
    MaskSpawnGroup=false
    Override Seed=false
    Rotation Axis=<0 0 1>
    Rotation Mode=::NewEnumerator1
    Shape Primitive=::NewEnumerator1
    Surface Only=true
    Use Endcaps In Surface Only Mode=false
    UseOwnerScale=true
   )
   &AddVelocity(Velocity=&Multiply_VectorByFloat(
     Vector=&Multiply_Vector(A=$Particles.ShapeLocation.ShapeVector, B=<1 1 0>)
     Float=&RandomRangeFloat(Minimum=30, Maximum=35)
    ))
   &AddVelocity(Velocity=&RandomRangeVector(Minimum=<0 0 50>, Maximum=<0 0 200>)) {module=AddVelocity001}
   &InheritVelocity(Inherited Velocity Amount Scale=<1 1 0.2>, Inherited Velocity Speed Limit=0, Speed Limit=false, UseSpeedThreshold=false)
  ]
  particleUpdateScript [
   &ParticleState()
   &SpriteRotationRate(Rotation Rate=&AngleConversion(
     Angle:NiagaraFloat=$Particles.SpriteRotationRate
     Angle Input:ENiagara_AngleInput=::NewEnumerator1
     Angle Output:ENiagara_AngleInput=::NewEnumerator0
    ))
   &ScaleSpriteSize(
    Scale Sprite Size Mode=::NewEnumerator0
    Uniform Scale Factor=&FloatFromCurve(FloatCurve=DataInterface() {propertyValues {Curve {keys [
         {arriveTangent=2.9377, interpMode=RCIM_Cubic, time=0, value=0.0414}
         {arriveTangent=0.9586, interpMode=RCIM_Cubic, leaveTangent=0.9586, time=0.1999, value=0.7634}
         <at 1 1>
        ]}}})
   )
   &ScaleColor(
    Color Value To Scale=$Particles.Color
    Scale Alpha=&FloatFromCurve(FloatCurve=DataInterface() {propertyValues {Curve {keys [<at 0.6 1 RCIM_Cubic>, {interpMode=RCIM_Cubic, leaveTangent=-4.9488, time=1, value=0}]}}})
    Scale RGB=&VectorFromCurve(VectorCurve=DataInterface() {propertyValues {XCurve {keys [<at 0 1>,<at 1 1>]}, YCurve {keys [<at 0 1>,<at 1 1>]}, ZCurve {keys [<at 0 1>,<at 1 1>]}}})
    ScaleA=true
    ScaleRGB=true
   )
   &SubUVAnimation(
    Random Start Frame=true
    "Sprite Renderer "=DataInterface() {propertyValues {SpriteRenderer=&self:TransSmoke_0.NiagaraSpriteRendererProperties_7}}
    SubUV Lookup Index Scale=0.25
    UseEndFrame=false
    UseStartFrame=false
   ) {module=SubUVAnimation001}
   &Drag(Drag=4, Use Linear Drag=true, Use Rotational Drag=true)
   &SolveForcesAndVelocity(Clamp Velocity=false, Limit Acceleration=false)
   &DynamicMaterialParameters(
    Index 0 Param 1=&FloatFromCurve(FloatCurve=DataInterface() {propertyValues {Curve {keys [<at 0.3 0.5 RCIM_Cubic>,<at 1 -1>]}}})
    Index 0 Param 3=&FloatFromCurve(FloatCurve=DataInterface() {propertyValues {Curve {keys [<at 0.5 0>,<at 1 0>]}}})
    Index 0 Param 4=&FloatFromCurve(FloatCurve=DataInterface() {propertyValues {Curve {keys [<at 0.6 1 RCIM_Cubic>,<at 1 0>]}}})
    Module.Index0Param0Write=true
    Module.Index0Param1Write=true
    Module.Index0Param2Write=true
    Module.Index0Param3Write=true
   )
  ]
  rendererClasses [&NiagaraSpriteRendererProperties]
  renderers [{rendererClass=&NiagaraSpriteRendererProperties}]
 }
}
summary {
 systemName=NS_Hoverboard_Athena_Land_Converted
 emitters [
  {emitterName=TransSmoke, rendererClasses [&NiagaraSpriteRendererProperties]}
  {bEnabled=false, emitterName=FoamRollingOut, rendererClasses [&NiagaraMeshRendererProperties]}
  {bEnabled=false, emitterName=RipplesSlow, rendererClasses [&NiagaraSpriteRendererProperties]}
  {emitterName=Ripple, rendererClasses [&NiagaraSpriteRendererProperties]}
  {emitterName=Light, rendererClasses [&NiagaraLightRendererProperties]}
  {emitterName=Lightning, rendererClasses [&NiagaraMeshRendererProperties]}
 ]
 userVariables [
  User.HBoardColorSet {type=NiagaraInt32}
  User.Moving {type=NiagaraFloat, defaultValue=0, description=""}
  User.Water {type=NiagaraFloat, defaultValue=1, description=""}
 ]
}
