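# Niagara system for hoverboard speed FX: flickering lightning arcs, front and back ripples, underside glows and camera speed lines.
# System.ColorSet picks arc and ripple colors and sets the glow meshes' visibility tag.
# Glows follow the skeletal mesh's front and back plate sockets; User.Fast feeds ripple materials, while speed-line spawn rate stays effectively constant.
# Five skid emitters (smoke, sharp, splash, GPU snow and water), driven by Skid, Snow and Water parameters, are disabled.
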NiagaraSystem /HoverboardV2/Vehicle/FX/NS_Hoverboard_Athena_SpeedFX_Converted
dependencies {
 usedDataInterfaces [
  &NiagaraDataInterfaceCurve
  &NiagaraDataInterfaceVector2DCurve
  &NiagaraDataInterfaceVectorCurve
  &NiagaraDataInterfaceSpriteRendererInfo
  &NiagaraDataInterfaceArrayColor
  &NiagaraDataInterfaceSkeletalMesh
 ]
 usedDynamicInputs [&ConvertVectorToPosition]
 usedModules [
  &SystemState
  &self:SystemSpawnScript.SystemScriptSource.SystemScriptGraph.NiagaraNodeAssignment_0.SetVariables_#1
 ]
}
emitters {
 LightningArcs {
  emitterUpdateScript [
   &EmitterState(Life Cycle Mode=::NewEnumerator1, Loop Behavior=::NewEnumerator1, Loop Delay=0.1, Loop Duration=0.5, UseLoopDelay=true)
   &SpawnBurst_Instantaneous(
    Spawn Count=&RandomRangeInt(Minimum=5, Maximum=6)
    Spawn Time=0.01
    Use Spawn Probability=false
    UseLoopCountLimit=false
   )
   &SpawnRate(
    SpawnRate=&Multiply_Float(
     A=20
     B=&FloatFromCurve(FloatCurve=DataInterface() {propertyValues {Curve {keys [<at 0 1 RCIM_Cubic>,<at 0.45 0>]}}}, CurveIndex=$Emitter.NormalizedLoopAge)
    )
    Use Spawn Probability=false
   )
  ]
  particleSpawnScript [
   &InitializeParticle(
    Color=&SelectLinearColorFromArray(
     Array Sampling Mode=::NewEnumerator1
     Direct Array Index=$System.ColorSet
     Color Selection Array=DataInterface() {propertyValues {ColorData [<rgba 0.5509 1.7594 2.5 1>, <rgba 0.4167 2 0.1 1>]}}
     Random Seed=?
    )
    Lifetime=&RandomRangeFloat(Minimum=0.2, Maximum=0.35)
    Mesh Scale=&RandomRangeVector(Minimum=<1 1 1>, Maximum=<2 2 2>)
    Mesh Scale Mode=::NewEnumerator4
    UsePositionOffset=false
    Write Material Random=true
   )
   &InitialMeshOrientation(
    Mesh Orientation Mode=::NewEnumerator7
    Rotation=&RandomRangeVector(Minimum=<0.1 0.1 1>, Maximum=<-0.1 -0.1 0>)
    Use Rotation Vector=true
   )
   &ShapeLocation(
    Cylinder Height=5
    Cylinder Radius=&FloatFromCurve(
     FloatCurve=DataInterface() {propertyValues {Curve {keys [{arriveTangent=-659.3079, interpMode=RCIM_Cubic, time=0, value=120}, <at 0.45 10>]}}}
     CurveIndex=$Emitter.LoopedAge
    )
    MaskSpawnGroup=false
    Offset=&RandomRangeVector(Minimum=<-40 0 5>, Maximum=<50 0 10>)
    Offset Mode=::NewEnumerator0
    Override Seed=false
    Rotation Axis=<0 0 1>
    Rotation Mode=::NewEnumerator1
    Shape Primitive=::NewEnumerator1
    Surface Only=true
    Use Endcaps In Surface Only Mode=false
    UseOwnerScale=true
   )
   &AddVelocity(Velocity=&Multiply_VectorByFloat(
     Vector=&Multiply_Vector(A=$Particles.ShapeLocation.ShapeVector, B=<1 1 0>)
     Float=&RandomRangeFloat(Minimum=-4, Maximum=-2)
    ))
  ]
  particleUpdateScript [
   &ParticleState()
   &ScaleMeshSize(Scale Factor=&VectorFromCurve(
     VectorCurve=DataInterface() {propertyValues {
       XCurve {keys [<at 0 1 RCIM_Cubic>,<at 0.7 1.2918 RCIM_Cubic>,<at 1 1>]}
       YCurve {keys [<at 0 1 RCIM_Cubic>,<at 0.7 1.7082 RCIM_Cubic>,<at 1 1>]}
       ZCurve {keys [<at 0 0.2 RCIM_Cubic>, {arriveTangent=-0.1, interpMode=RCIM_Cubic, leaveTangent=-0.1, time=0.7, value=1}, <at 1 0.1>]}
      }}
     Scale Curve=&VectorFromFloat(Value=1)
    )) {module=ScaleMeshSize001}
   &UpdateMeshOrientation(Rotation Vector=<0 0 1>)
   &Drag(Use Linear Drag=true, Use Rotational Drag=true)
   &InheritSourceMovement(
    Applied Position Scale=&VectorFromCurve(VectorCurve=DataInterface() {propertyValues {
       XCurve {keys [<at 0 1 RCIM_Cubic>, {arriveTangent=-1.1983, tangentMode=RCTM_User, time=1, value=0.25}]}
       YCurve {keys [<at 0 1 RCIM_Cubic>, {arriveTangent=-1.1983, tangentMode=RCTM_User, time=1, value=0.25}]}
       ZCurve {keys [<at 0 0>,<at 1 0>]}
      }})
    ApplyForce=false
    ApplyPosition=true
    ApplyPositionFalloff=false
   )
   &SolveForcesAndVelocity(Clamp Velocity=false, Limit Acceleration=false)
   &CameraOffset(Camera Offset Amount=25)
   &ScaleColor(
    Color Value To Scale=$Particles.Color
    Scale Alpha=&FloatFromCurve(FloatCurve=DataInterface() {propertyValues {Curve {keys [
         <at 0 0.5 RCIM_Cubic>
         {arriveTangent=1.26, interpMode=RCIM_Cubic, leaveTangent=1.26, time=0.25, value=1}
         {arriveTangent=-3.3413, interpMode=RCIM_Cubic, leaveTangent=-3.3413, time=0.4927, value=1.1208}
         {arriveTangent=0.3732, interpMode=RCIM_Cubic, leaveTangent=0.3732, time=0.5399, value=0.0315}
         {arriveTangent=-0.022, interpMode=RCIM_Cubic, leaveTangent=-0.022, time=0.5887, value=1.1566}
         {arriveTangent=0.0713, interpMode=RCIM_Cubic, leaveTangent=0.0713, time=0.6342, value=0.0294}
         {arriveTangent=-0.1642, interpMode=RCIM_Cubic, leaveTangent=-0.1642, time=0.6679, value=1.1623}
         {arriveTangent=-4.6994, interpMode=RCIM_Cubic, leaveTangent=-4.6994, time=0.7083, value=0.0172}
         {arriveTangent=-0.0103, interpMode=RCIM_Cubic, leaveTangent=-0.0103, time=0.7605, value=0.727}
         {arriveTangent=-0.453, interpMode=RCIM_Cubic, leaveTangent=-0.453, time=0.8161, value=0.0161}
         {arriveTangent=0.0387, interpMode=RCIM_Cubic, leaveTangent=0.0387, time=0.8565, value=0.6835}
         {arriveTangent=-2.9465, interpMode=RCIM_Cubic, leaveTangent=-2.9465, time=0.9045, value=0.0196}
         {arriveTangent=-0.2048, interpMode=RCIM_Cubic, leaveTangent=-0.2048, time=0.9441, value=0.4254}
         <at 1 0>
        ]}}})
    ScaleA=true
    ScaleRGB=true
   )
   &DynamicMaterialParameters(
    Index 0 Param 1=&FloatFromCurve(FloatCurve=DataInterface() {propertyValues {Curve {keys [
         {arriveTangent=45.9263, interpMode=RCIM_Cubic, time=0, value=0}
         {arriveTangent=3.3333, interpMode=RCIM_Cubic, leaveTangent=3.3333, time=0.1, value=1.5}
         <at 0.3 1>
         {arriveTangent=-1.4286, interpMode=RCIM_Cubic, leaveTangent=-1.4286, time=0.7, value=1}
         <at 1 0>
        ]}}})
    Index 0 Param 2=&FloatFromCurve(FloatCurve=DataInterface() {propertyValues {Curve {keys [<at 0 0.4 RCIM_Cubic>, {arriveTangent=-0.4, interpMode=RCIM_Cubic, leaveTangent=-0.4, time=0.8, value=2}, <at 1 0>]}}})
    Index 0 Param 3=&RandomRangeFloat(Minimum=0, Maximum=1)
    Index 0 Param 4=&RandomRangeFloat(Minimum=0, Maximum=1)
    Index 1 Param 1=&MakeFloatFromVector(Vector=Linked() {linkedVariable {name=Engine.Owner.Position, type=NiagaraPosition}}, Channel=::NewEnumerator0)
    Index 1 Param 2=&MakeFloatFromVector(Vector=Linked() {linkedVariable {name=Engine.Owner.Position, type=NiagaraPosition}}, Channel=::NewEnumerator1)
    Index 1 Param 3=&MakeFloatFromVector(Vector=Linked() {linkedVariable {name=Engine.Owner.Position, type=NiagaraPosition}}, Channel=::NewEnumerator2)
    Index 1 Param 4=10
    Module.Index0Param0Write=true
    Module.Index0Param1Write=true
    Module.Index0Param2Write=true
    Module.Index0Param3Write=true
    Module.Index1Param0Write=true
    Module.Index1Param1Write=true
    Module.Index1Param2Write=true
    Module.Index1Param3Write=true
    Write Parameter Index 1=true
   ) {module=DynamicMaterialParameters002}
  ]
  rendererClasses [&NiagaraMeshRendererProperties]
  renderers [{rendererClass=&NiagaraMeshRendererProperties, rendererIndex=0}]
 }
 RipplesBack {
  emitterUpdateScript [
   &EmitterState(
    Life Cycle Mode=::NewEnumerator1
    Loop Duration=&RandomRangeFloat(Minimum=0.3, Maximum=0.35, Evaluation Type=::NewEnumerator1)
    Recalculate Duration Each Loop=true
    UseLoopDelay=true
   )
   &SpawnBurst_Instantaneous(Use Spawn Probability=false, UseLoopCountLimit=false)
  ]
  particleSpawnScript [
   &InitializeParticle(
    Color=&SelectLinearColorFromArray(
     Array Sampling Mode=::NewEnumerator1
     Direct Array Index=$System.ColorSet
     Color Selection Array=DataInterface() {propertyValues {ColorData [<rgba 0.3642 0.6292 1 1>, <rgba 0.3642 1 0.3642 1>]}}
     Random Seed=?
    )
    Lifetime=&RandomRangeFloat(Minimum=0.75, Maximum=0.8)
    Position Offset=<25 0 -45>
    Sprite Rotation Angle=0.25
    Sprite Rotation Mode=::NewEnumerator1
    Sprite Size=<80 80>
    Sprite Size Mode=::NewEnumerator4
    UsePositionOffset=true
    Write Material Random=true
   )
   &DynamicMaterialParameters(
    Index 0 Param 1=&RandomRangeFloat(Minimum=0, Maximum=1)
    Index 0 Param 2=&Lerp_Float(A=0, B=0.2, Alpha=&MakeFloatFromVector(Vector=$User.Fast, Channel=::NewEnumerator0))
    Index 0 Param 3=1
    Index 0 Param 4=1
    Module.Index0Param0Write=true
    Module.Index0Param1Write=true
    Module.Index0Param2Write=true
    Module.Index0Param3Write=true
   ) {module=DynamicMaterialParameters002}
   &CameraOffset(Camera Offset Amount=18)
  ]
  particleUpdateScript [
   /SetVariables(Particles.Position=&AddVectorToPosition(Position=$User.Front, Vector=<12 0 0>)) {enabled=false, isSetParameters=true, module=SetVariables_#2}
   &ParticleState()
   &ScaleColor(
    Scale Alpha=&FloatFromCurve(FloatCurve=DataInterface() {propertyValues {Curve {keys [<at 0 0 RCIM_Cubic>, <at 0.25 1>, {arriveTangent=-1.3333, interpMode=RCIM_Cubic, leaveTangent=-1.3333, time=0.75, value=1}, <at 1 0>]}}})
    ScaleA=true
    ScaleRGB=true
   ) {module=ScaleColor001}
   /SetVariables(Particles.SpriteFacing=&TransformVector(Vector=<0 0 1>, Source Space=::World, Destination Space=::Simulation)) {isSetParameters=true, module=SetVariables_#3}
  ]
  rendererClasses [&NiagaraSpriteRendererProperties]
  renderers [{rendererClass=&NiagaraSpriteRendererProperties}]
 }
 RipplesFront {
  emitterUpdateScript [
   &EmitterState(
    Life Cycle Mode=::NewEnumerator1
    Loop Duration=&RandomRangeFloat(Minimum=0.3, Maximum=0.35, Evaluation Type=::NewEnumerator1)
    Recalculate Duration Each Loop=true
    UseLoopDelay=true
   )
   &SpawnBurst_Instantaneous(Use Spawn Probability=false, UseLoopCountLimit=false)
  ]
  particleSpawnScript [
   &InitializeParticle(
    Color=&SelectLinearColorFromArray(
     Array Sampling Mode=::NewEnumerator1
     Direct Array Index=$System.ColorSet
     Color Selection Array=DataInterface() {propertyValues {ColorData [<rgba 0.3642 0.6292 1 1>, <rgba 0.3642 1 0.3642 1>]}}
     Random Seed=?
    )
    Lifetime=&RandomRangeFloat(Minimum=0.75, Maximum=0.8)
    Position Offset=<-25 0 -45>
    Sprite Rotation Angle=0.25
    Sprite Rotation Mode=::NewEnumerator1
    Sprite Size=<80 80>
    Sprite Size Mode=::NewEnumerator4
    UsePositionOffset=true
    Write Material Random=true
   )
   &DynamicMaterialParameters(
    Index 0 Param 1=&RandomRangeFloat(Minimum=0, Maximum=1)
    Index 0 Param 2=&Lerp_Float(A=0, B=0.2, Alpha=&MakeFloatFromVector(Vector=$User.Fast, Channel=::NewEnumerator0))
    Index 0 Param 3=1
    Index 0 Param 4=1
    Module.Index0Param0Write=true
    Module.Index0Param1Write=true
    Module.Index0Param2Write=true
    Module.Index0Param3Write=true
   ) {module=DynamicMaterialParameters002}
   &CameraOffset(Camera Offset Amount=18)
  ]
  particleUpdateScript [
   /SetVariables(Particles.Position=&AddVectorToPosition(Position=$User.Front, Vector=<12 0 0>)) {enabled=false, isSetParameters=true, module=SetVariables_#2}
   &ParticleState()
   &ScaleColor(
    Scale Alpha=&FloatFromCurve(FloatCurve=DataInterface() {propertyValues {Curve {keys [<at 0 0 RCIM_Cubic>, <at 0.25 1>, {arriveTangent=-1.3333, interpMode=RCIM_Cubic, leaveTangent=-1.3333, time=0.75, value=1}, <at 1 0>]}}})
    ScaleA=true
    ScaleRGB=true
   ) {module=ScaleColor001}
   /SetVariables(Particles.SpriteFacing=&TransformVector(Vector=<0 0 1>, Source Space=::World, Destination Space=::Simulation)) {isSetParameters=true, module=SetVariables_#3}
  ]
  rendererClasses [&NiagaraSpriteRendererProperties]
  renderers [{rendererClass=&NiagaraSpriteRendererProperties}]
 }
 SPlashSkid {
  bEnabled=false
  emitterUpdateScript [
   &EmitterState(
    Life Cycle Mode=::NewEnumerator1
    Loop Duration=&RandomRangeFloat(Minimum=0.5, Maximum=1.2, Evaluation Type=::NewEnumerator1)
    Recalculate Duration Each Loop=true
    UseLoopDelay=true
   )
   &SpawnRate(
    SpawnRate=&Lerp_Float(A=0, B=80, Alpha=&Multiply_Float(A=$User.Skid, B=$User.Water))
    Use Spawn Probability=false
   )
  ]
  particleSpawnScript [
   &InitializeParticle(
    Color Maximum=<rgba 0.6033 0.8466 1 0.6>
    Color Minimum=<rgba 0.5 0.5 0.5 0.4>
    Color Mode=::NewEnumerator2
    Lifetime=&RandomRangeFloat(Minimum=0.15, Maximum=0.25)
    Mesh Scale=&Multiply_VectorByFloat(Vector=&RandomRangeVector(Minimum=<1.4 1 1>, Maximum=<2.5 1 1>), Float=2)
    Mesh Scale Mode=::NewEnumerator4
    UsePositionOffset=false
    Write Material Random=true
   )
   &AddVelocity(Velocity=&RandomRangeVector(Minimum=<0 0 400>, Maximum=<0 0 300>))
   &ShapeLocation(
    Cylinder Height=60
    Cylinder Radius=5
    MaskSpawnGroup=false
    Offset=<25 0 -10>
    Offset Mode=::NewEnumerator0
    Override Seed=false
    Rotation Mode=::NewEnumerator1
    Shape Primitive=::NewEnumerator1
    Surface Only=true
    Use Endcaps In Surface Only Mode=false
    UseOwnerScale=true
   )
   &AddVelocity(Velocity=&Multiply_VectorByFloat(Vector=&Multiply_Vector(A=$Particles.ShapeLocation.ShapeVector, B=<0.5 1 1>), Float=8)) {module=AddVelocity001}
   &AddVelocity(Velocity=&RandomRangeVector(Minimum=<-400 25 25>, Maximum=<-500 -25 25>)) {module=AddVelocity002}
   &InheritVelocity(Inherited Velocity Amount Scale=<0.5 0.5 0.5>, Inherited Velocity Speed Limit=0, Speed Limit=false, UseSpeedThreshold=false)
   &CameraOffset(Camera Offset Amount=-20)
  ]
  particleUpdateScript [
   &ParticleState()
   &ScaleMeshSize(Scale Factor=&VectorFromCurve(VectorCurve=DataInterface() {propertyValues {
       XCurve {keys [
         {arriveTangent=4.58, interpMode=RCIM_Cubic, time=0, value=0}
         {arriveTangent=1, interpMode=RCIM_Cubic, leaveTangent=1, time=0.1482, value=0.4348}
         {interpMode=RCIM_Cubic, leaveTangent=0.2409, time=1, value=1}
        ]}
       YCurve {keys [<at 0 1 RCIM_Cubic>,<at 0.1482 1 RCIM_Cubic>,<at 1 1 RCIM_Cubic>]}
       ZCurve {keys [<at 0 1 RCIM_Cubic>,<at 0.1482 1 RCIM_Cubic>,<at 1 1 RCIM_Cubic>]}
      }})) {module=ScaleMeshSize001}
   &Drag(Drag=6, Use Linear Drag=true, Use Rotational Drag=true)
   &AccelerationForce(Acceleration=<0 0 -500>)
   &SolveForcesAndVelocity(Clamp Velocity=false, Limit Acceleration=false)
   &DynamicMaterialParameters(
    Index 0 Param 1=&FloatFromCurve(FloatCurve=DataInterface() {propertyValues {Curve {keys [
         {arriveTangent=2.5029, interpMode=RCIM_Cubic, leaveTangent=2.5029, tangentMode=RCTM_User, time=0.1, value=0.1}
         {arriveTangent=3.1056, tangentMode=RCTM_User, time=0.8, value=1}
        ]}}})
    Index 0 Param 2=&RandomRangeFloat(Minimum=2, Maximum=3)
    Index 0 Param 3=&RandomRangeFloat(Minimum=0, Maximum=1)
    Index 0 Param 4=&FloatFromCurve(
     FloatCurve=DataInterface() {propertyValues {Curve {keys [{arriveTangent=5.5011, interpMode=RCIM_Cubic, time=0.3, value=0}, {interpMode=RCIM_Cubic, leaveTangent=1.1181, time=0.6, value=1}]}}}
     Scale Curve=2.5
    )
    Module.Index0Param0Write=true
    Module.Index0Param1Write=true
    Module.Index0Param2Write=true
    Module.Index0Param3Write=true
   ) {module=DynamicMaterialParameters002}
   &ScaleColor(
    Scale Alpha=&FloatFromCurve(
     FloatCurve=DataInterface() {propertyValues {Curve {keys [
         <at 0.3 0 RCIM_Cubic>
         {arriveTangentWeight=0.2236, interpMode=RCIM_Cubic, leaveTangentWeight=0.1491, tangentMode=RCTM_User, time=0.6, value=1}
         {arriveTangentWeight=0.1795, interpMode=RCIM_Cubic, leaveTangentWeight=0.1795, tangentMode=RCTM_User, time=0.8, value=1}
         {arriveTangent=-8.3504, interpMode=RCIM_Cubic, leaveTangent=-8.3504, tangentMode=RCTM_User, time=1, value=0}
        ]}}}
     Scale Curve=50
    )
    Scale RGB=&VectorFromCurve(VectorCurve=DataInterface() {propertyValues {
       XCurve {keys [<at 0 0.4248 RCIM_Cubic>,<at 0.25 1>]}
       YCurve {keys [<at 0 1.472 RCIM_Cubic>,<at 0.25 1>]}
       ZCurve {keys [<at 0 2 RCIM_Cubic>,<at 0.25 1>]}
      }})
    ScaleA=true
    ScaleRGB=true
   ) {module=ScaleColor001}
  ]
  rendererClasses [&NiagaraMeshRendererProperties]
  renderers [{rendererClass=&NiagaraMeshRendererProperties, rendererIndex=0}]
 }
 SharpSkid {
  bEnabled=false
  emitterUpdateScript [
   &EmitterState(
    Life Cycle Mode=::NewEnumerator1
    Loop Duration=&RandomRangeFloat(Minimum=0.5, Maximum=1.2, Evaluation Type=::NewEnumerator1)
    Recalculate Duration Each Loop=true
    UseLoopDelay=true
   )
   &SpawnRate(
    SpawnRate=&Lerp_Float(
     A=0
     B=80
     Alpha=&Multiply_Float(A=$User.Skid, B=&OneMinusFloat(Float=$User.Water))
    )
    Use Spawn Probability=false
   )
  ]
  particleSpawnScript [
   &InitializeParticle(
    Color=/Niagara/DynamicInputs/Lerp/SmoothLerpOverTime_Color(Smooth Value:LinearColor=<rgba 1.5 1.5 1.5 0.4>, Target Value:LinearColor=<rgba 1.81 2.5399 3 0.6>, Convergence Rate:NiagaraFloat=$User.Snow)
    Lifetime=&RandomRangeFloat(Minimum=0.15, Maximum=0.25)
    Mesh Scale=&Multiply_VectorByFloat(Vector=&RandomRangeVector(Minimum=<1.4 1 1>, Maximum=<2.5 1 1>), Float=2)
    Mesh Scale Mode=::NewEnumerator4
    UsePositionOffset=false
    Write Material Random=true
   )
   &AddVelocity(Velocity=&RandomRangeVector(Minimum=<0 0 100>, Maximum=<0 0 200>))
   &ShapeLocation(
    Cylinder Height=60
    Cylinder Radius=5
    MaskSpawnGroup=false
    Offset=<25 0 -10>
    Offset Mode=::NewEnumerator0
    Override Seed=false
    Rotation Mode=::NewEnumerator1
    Shape Primitive=::NewEnumerator1
    Surface Only=true
    Use Endcaps In Surface Only Mode=false
    UseOwnerScale=true
   )
   &AddVelocity(Velocity=&Multiply_VectorByFloat(Vector=&Multiply_Vector(A=$Particles.ShapeLocation.ShapeVector, B=<0.5 1 1>), Float=8)) {module=AddVelocity001}
   &AddVelocity(Velocity=&RandomRangeVector(Minimum=<-400 25 25>, Maximum=<-500 -25 25>)) {module=AddVelocity002}
   &InheritVelocity(Inherited Velocity Amount Scale=<0.5 0.5 0.5>, Inherited Velocity Speed Limit=0, Speed Limit=false, UseSpeedThreshold=false)
   &CameraOffset(Camera Offset Amount=-20)
  ]
  particleUpdateScript [
   &ParticleState()
   &ScaleMeshSize(Scale Factor=&VectorFromCurve(VectorCurve=DataInterface() {propertyValues {
       XCurve {keys [
         {arriveTangent=4.58, interpMode=RCIM_Cubic, time=0, value=0}
         {arriveTangent=1, interpMode=RCIM_Cubic, leaveTangent=1, time=0.1482, value=0.4348}
         {interpMode=RCIM_Cubic, leaveTangent=0.2409, time=1, value=1}
        ]}
       YCurve {keys [<at 0 1 RCIM_Cubic>,<at 0.1482 1 RCIM_Cubic>,<at 1 1 RCIM_Cubic>]}
       ZCurve {keys [<at 0 1 RCIM_Cubic>,<at 0.1482 1 RCIM_Cubic>,<at 1 1 RCIM_Cubic>]}
      }})) {module=ScaleMeshSize001}
   &Drag(Drag=6, Use Linear Drag=true, Use Rotational Drag=true)
   &SolveForcesAndVelocity(Clamp Velocity=false, Limit Acceleration=false)
   &DynamicMaterialParameters(
    Index 0 Param 1=&FloatFromCurve(FloatCurve=DataInterface() {propertyValues {Curve {keys [<at 0.1 0.1 RCIM_Cubic>,<at 0.8 1>]}}})
    Index 0 Param 2=&RandomRangeFloat(Minimum=2, Maximum=3)
    Index 0 Param 3=&RandomRangeFloat(Minimum=0, Maximum=1)
    Index 0 Param 4=&FloatFromCurve(FloatCurve=DataInterface() {propertyValues {Curve {keys [{arriveTangent=5.5011, interpMode=RCIM_Cubic, time=0, value=0.25}, {interpMode=RCIM_Cubic, leaveTangent=1.1181, time=1, value=3}]}}})
    Module.Index0Param0Write=true
    Module.Index0Param1Write=true
    Module.Index0Param2Write=true
    Module.Index0Param3Write=true
   ) {module=DynamicMaterialParameters002}
   &ScaleColor(
    Scale Alpha=&FloatFromCurve(FloatCurve=DataInterface() {propertyValues {Curve {keys [<at 0 0 RCIM_Cubic>, <at 0.2 1 RCIM_Cubic>, {interpMode=RCIM_Cubic, leaveTangent=-1.9093, time=1, value=0}]}}})
    Scale RGB=&VectorFromCurve(VectorCurve=DataInterface() {propertyValues {
       XCurve {keys [<at 0 0.4248 RCIM_Cubic>,<at 0.25 1>]}
       YCurve {keys [<at 0 1.472 RCIM_Cubic>,<at 0.25 1>]}
       ZCurve {keys [<at 0 2 RCIM_Cubic>,<at 0.25 1>]}
      }})
    ScaleA=true
    ScaleRGB=true
   ) {module=ScaleColor001}
  ]
  rendererClasses [&NiagaraMeshRendererProperties]
  renderers [{rendererClass=&NiagaraMeshRendererProperties, rendererIndex=0}]
 }
 SmokeSkid {
  bEnabled=false
  emitterUpdateScript [
   &EmitterState(Life Cycle Mode=::NewEnumerator1, Loop Duration=1, UseLoopDelay=true)
   &SpawnRate(
    SpawnRate=&Lerp_Float(
     A=0
     B=80
     Alpha=&Multiply_Float(A=$User.Skid, B=&OneMinusFloat(Float=$User.Water))
    )
    Use Spawn Probability=false
   )
  ]
  particleSpawnScript [
   &InitializeParticle(
    Color=&LerpLinearColors(
     StartColor=&RandomRangeLinearColor(Minimum=<rgba 1 1 1 1>, Maximum=<rgba 0.4 0.4 0.4 1>)
     EndColor=&RandomRangeLinearColor(Minimum=<rgba 1 1 1 1>, Maximum=<rgba 0.4 0.4 0.4 1>)
     LerpFactor=$User.Snow
    )
    Lifetime=&RandomRangeFloat(Minimum=0.22, Maximum=0.35)
    Position Offset:NiagaraPosition=$User.Front
    Sprite Size=&RandomRangeVector2D(Minimum=<40 60>, Maximum=<70 90>)
    Sprite Size Mode=::NewEnumerator4
    UsePositionOffset=true
    Write Material Random=true
   )
   &AddVelocity(Velocity=&RandomRangeVector(Minimum=<0 0 400>, Maximum=<0 0 500>))
   &InheritVelocity(Inherited Velocity Speed Limit=0, Speed Limit=false, UseSpeedThreshold=false)
   &ShapeLocation(
    Cylinder Height=60
    Cylinder Radius=5
    MaskSpawnGroup=false
    Offset=<10 0 -15>
    Offset Mode=::NewEnumerator0
    Override Seed=false
    Rotation Mode=::NewEnumerator1
    Shape Primitive=::NewEnumerator1
    Surface Only=false
    UseOwnerScale=true
   )
   &AddVelocity(Velocity=&Multiply_VectorByFloat(Vector=$Particles.ShapeLocation.ShapeVector, Float=6)) {module=AddVelocity001}
   &AddVelocity(Velocity=&RandomRangeVector(Minimum=<-700 0 25>, Maximum=<-800 1 25>)) {module=AddVelocity002}
  ]
  particleUpdateScript [
   &ParticleState()
   &ScaleSpriteSize(
    Scale Factor=&Vector2DFromCurve(
     Vector2Curve=DataInterface() {propertyValues {
       XCurve {keys [
         {arriveTangent=0.4028, interpMode=RCIM_Cubic, time=0, value=0.1}
         {arriveTangent=1.1, interpMode=RCIM_Cubic, leaveTangent=1.1, time=0.3979, value=0.6219}
         {interpMode=RCIM_Cubic, leaveTangent=-0.395, time=1, value=1.2}
        ]}
       YCurve {keys [
         {arriveTangent=3.678, interpMode=RCIM_Cubic, time=0, value=0}
         {arriveTangent=0.5, interpMode=RCIM_Cubic, leaveTangent=0.5, time=0.3979, value=0.9996}
         {interpMode=RCIM_Cubic, leaveTangent=-1.8217, time=1, value=0.5}
        ]}
      }}
     ScaleCurve=&Multiply_Vector2D_ByFloat(Float=0.15, Vector2D=<1 1>)
    )
    Scale Sprite Size Mode=::NewEnumerator2
   )
   &ScaleColor(
    Scale Alpha=&FloatFromCurve(
     FloatCurve=DataInterface() {propertyValues {Curve {keys [<at 0 0 RCIM_Cubic>,<at 0.1 0.25 RCIM_Cubic>,<at 1 0 RCIM_Cubic>]}}}
     Scale Curve=0.65
    )
    Scale RGB=&VectorFromCurve(VectorCurve=DataInterface() {propertyValues {
       XCurve {keys [<at 0 0.2273 RCIM_Cubic>,<at 0.2 1>]}
       YCurve {keys [<at 0 1.1261 RCIM_Cubic>,<at 0.2 1>]}
       ZCurve {keys [<at 0 2 RCIM_Cubic>,<at 0.2 1>]}
      }})
    ScaleA=true
    ScaleRGB=true
   ) {module=ScaleColor001}
   &SubUVAnimation(
    "Sprite Renderer "=DataInterface() {propertyValues {SpriteRenderer=&self:SmokeSkid_2.NiagaraSpriteRendererProperties_8}}
    SubUV Lookup Index Scale=0.6667
    UseEndFrame=false
    UseStartFrame=false
   ) {module=SubUVAnimation001}
   &Drag(Drag=2, Use Linear Drag=true, Use Rotational Drag=true)
   &SolveForcesAndVelocity(Clamp Velocity=false, Limit Acceleration=false)
   &ScaleSpriteSizeBySpeed(Max Scale Factor=<10 20>, Source Velocity:Vector3f=&Multiply_Vector(A=<0 0.01 1>, B=$Particles.Velocity))
   &DynamicMaterialParameters(
    Index 0 Param 1=&FloatFromCurve(FloatCurve=DataInterface() {propertyValues {Curve {keys [<at 0.2 1 RCIM_Cubic>,<at 1 -1>]}}})
    Index 0 Param 3=&FloatFromCurve(FloatCurve=DataInterface() {propertyValues {Curve {keys [<at 0.5 0>,<at 1 0>]}}})
    Index 0 Param 4=&FloatFromCurve(FloatCurve=DataInterface() {propertyValues {Curve {keys [<at 0.2 1 RCIM_Cubic>,<at 1 0.5>]}}})
    Module.Index0Param0Write=true
    Module.Index0Param1Write=true
    Module.Index0Param2Write=true
    Module.Index0Param3Write=true
   )
   &CameraOffset(Camera Offset Amount=-20)
  ]
  rendererClasses [&NiagaraSpriteRendererProperties]
  renderers [{rendererClass=&NiagaraSpriteRendererProperties}]
 }
 SnowSkid {
  bEnabled=false
  simTarget=GPUComputeSim
  emitterUpdateScript [
   &EmitterState(Life Cycle Mode=::NewEnumerator1, Loop Duration=2, UseLoopDelay=true)
   &SpawnRate(
    SpawnRate=&Lerp_Float(A=0, B=1500, Alpha=&Multiply_Float(A=$User.Skid, B=$User.Snow))
    Use Spawn Probability=false
   )
  ]
  particleSpawnScript [
   &InitializeParticle(
    Color Maximum=<rgba 1.374 1.4375 1.5 1.25>
    Color Minimum=<rgba 1.374 1.4375 1.5 0.75>
    Color Mode=::NewEnumerator2
    Lifetime=&RandomRangeFloat(Minimum=0.1, Maximum=0.6)
    Position Offset:NiagaraPosition=$User.Front
    Sprite Size=&RandomRangeVector2D(Minimum=<3 8>, Maximum=<12 24>)
    Sprite Size Mode=::NewEnumerator4
    UsePositionOffset=true
    Write Material Random=true
   )
   &ShapeLocation(
    Cylinder Height=60
    Cylinder Radius=5
    MaskSpawnGroup=false
    Offset=<10 0 -15>
    Offset Mode=::NewEnumerator0
    Override Seed=false
    Rotation Mode=::NewEnumerator1
    Shape Primitive=::NewEnumerator1
    Surface Only=true
    Use Endcaps In Surface Only Mode=false
    UseOwnerScale=true
   )
   &AddVelocity(Velocity=&Multiply_VectorByFloat(Vector=&Multiply_Vector(A=$Particles.ShapeLocation.ShapeVector, B=<0.8 1 1>), Float=25)) {module=AddVelocity001}
   &InheritVelocity(Inherited Velocity Amount Scale=<2 2 1>, Inherited Velocity Speed Limit=0, Speed Limit=false, UseSpeedThreshold=false)
   &AddVelocity(Velocity=&RandomRangeVector(Minimum=<0 0 400>, Maximum=<0 0 600>), Velocity Speed Scale=1.5)
   &AddVelocity(Velocity=&RandomRangeVector(Minimum=<-850 0 25>, Maximum=<-900 0 0>)) {module=AddVelocity002}
  ]
  particleUpdateScript [
   &ParticleState()
   &ScaleSpriteSize(
    Scale Factor=&Vector2DFromCurve(
     Vector2Curve=DataInterface() {propertyValues {
       XCurve {keys [
         {arriveTangent=5.7886, interpMode=RCIM_Cubic, time=0, value=0.2}
         {arriveTangent=0.8, interpMode=RCIM_Cubic, leaveTangent=0.8, time=0.2, value=0.75}
         <at 1 1 RCIM_Cubic>
        ]}
       YCurve {keys [
         {arriveTangent=9.9545, interpMode=RCIM_Cubic, time=0, value=0}
         {arriveTangent=0.3, interpMode=RCIM_Cubic, leaveTangent=0.3, time=0.2, value=0.9996}
         <at 1 0.3 RCIM_Cubic>
        ]}
      }}
     ScaleCurve=<1 1>
    )
    Scale Sprite Size Mode=::NewEnumerator2
   ) {module=ScaleSpriteSize001}
   &AccelerationForce(Acceleration=<0 0 -1600>)
   &SolveForcesAndVelocity(Clamp Velocity=false, Limit Acceleration=false)
   &ScaleSpriteSizeBySpeed(
    Max Scale Factor=<2 4>
    Min Scale Factor=<1 1.01>
    Source Velocity:Vector3f=&Multiply_Vector(A=<0 0.01 1>, B=$Particles.Velocity)
   )
   &ScaleColor(
    Scale Alpha=&FloatFromCurve(FloatCurve=DataInterface() {propertyValues {Curve {keys [<at 0 0 RCIM_Cubic>,<at 0.2 1 RCIM_Cubic>,<at 1 0 RCIM_Cubic>]}}})
    ScaleA=true
    ScaleRGB=true
   ) {module=ScaleColor001}
  ]
  rendererClasses [&NiagaraSpriteRendererProperties]
  renderers [{rendererClass=&NiagaraSpriteRendererProperties}]
 }
 SpeedLinesCamera {
  emitterUpdateScript [
   &EmitterState(Life Cycle Mode=::NewEnumerator1, Loop Duration=1, UseLoopDelay=true)
   &SpawnRate(SpawnRate=300, Use Spawn Probability=false) {enabled=false, module=SpawnRate001}
   &SpawnRate(
    SpawnRate=&Lerp_Float(
     A=300
     B=300
     Alpha=&FloatFromCurve(
      FloatCurve=DataInterface() {propertyValues {Curve {keys [<at 0 0.05>,<at 1 2>]}}}
      CurveIndex=&MakeFloatFromVector(Vector=$User.Fast, Channel=::NewEnumerator0)
     )
    )
    Use Spawn Probability=false
   )
  ]
  particleSpawnScript [
   &InitializeParticle(
    Lifetime=&RandomRangeFloat(Minimum=0.25, Maximum=0.35)
    Sprite Size=&RandomRangeVector2D(Minimum=<4 2>, Maximum=<5.5 4>)
    Sprite Size Mode=::NewEnumerator4
    UsePositionOffset=false
    Write Material Random=true
   )
   &InheritVelocity(Inherited Velocity Amount Scale=<-0.7 -0.7 -0.7>, Inherited Velocity Speed Limit=0, Speed Limit=false, UseSpeedThreshold=false)
   &ShapeLocation(
    Cylinder Radius=250
    MaskSpawnGroup=false
    Offset=<600 0 70>
    Offset Mode=::NewEnumerator0
    Override Seed=false
    Rotation Mode=::NewEnumerator1
    Shape Primitive=::NewEnumerator1
    Surface Only=true
    Use Endcaps In Surface Only Mode=false
    UseOwnerScale=true
   )
   &AddVelocity(Velocity=&Multiply_VectorByFloat(Vector=&Multiply_Vector(A=$Particles.ShapeLocation.ShapeVector, B=<0 1 1>), Float=1.25)) {module=AddVelocity001}
   &Color(Color=&MakeLinearColorFromVectorAndFloat(
     "Vector (RGB)"=<0.6875 0.6851 0.6839>
     "Float (Alpha)"=&ClampFloat(Float=&RandomRangeFloat(Minimum=0.1, Maximum=0.2))
    ))
   &AddVelocity(Velocity=&RandomRangeVector(Minimum=<-70 0 0>, Maximum=<-25 0 0>))
   &CameraOffset(Camera Offset Amount=-100)
  ]
  particleUpdateScript [
   &ParticleState()
   &ScaleSpriteSize(
    Scale Factor=&Vector2DFromCurve(
     Vector2Curve=DataInterface() {propertyValues {XCurve {keys [<at 0 0.5 RCIM_Cubic>,<at 0.5 1 RCIM_Cubic>,<at 1 0.5>]}, YCurve {keys [<at 0 0 RCIM_Cubic>,<at 0.5 1 RCIM_Cubic>,<at 1 0>]}}}
     ScaleCurve=<1 1>
    )
    Scale Sprite Size Mode=::NewEnumerator2
   )
   &SolveForcesAndVelocity(Clamp Velocity=false, Limit Acceleration=false)
   &ScaleSpriteSizeBySpeed(
    Max Scale Factor=<1 100>
    Min Scale Factor=<1 1.05>
    Source Velocity:Vector3f=&Multiply_Vector(A=<0 0.05 1>, B=$Particles.Velocity)
   )
   &ScaleColor(
    Color Value To Scale=$Particles.Color
    Scale Alpha=&FloatFromCurve(FloatCurve=DataInterface() {propertyValues {Curve {keys [<at 0 0 RCIM_Cubic>,<at 0.4449 0.904 RCIM_Cubic>,<at 1 0>]}}}, Scale Curve=1.25)
    ScaleA=true
    ScaleRGB=true
   )
  ]
  rendererClasses [&NiagaraSpriteRendererProperties]
  renderers [{rendererClass=&NiagaraSpriteRendererProperties}]
 }
 UndersideGlows {
  emitterUpdateScript [
   &EmitterState(Inactive Response=::NewEnumerator1, Life Cycle Mode=::NewEnumerator1, Loop Duration=0.8, UseLoopDelay=true)
   &SpawnBurst_Instantaneous(Spawn Count=2, Use Spawn Probability=false, UseLoopCountLimit=false)
  ]
  particleSpawnScript [
   &InitializeParticle(Color=<rgba 10 10 10 1>, Mesh Scale=<0.25 0.3 0.3>, Mesh Scale Mode=::NewEnumerator4, UsePositionOffset=false, Write Material Random=true)
   &InitialMeshOrientation(Mesh Orientation Mode=::NewEnumerator7, Rotation=<0 -0.25 0>, Use Rotation Vector=true)
   &InitialMeshOrientation(
    Mesh Orientation Mode=::NewEnumerator7
    Rotation=&RandomRangeVector(Minimum=<0 0 0>, Maximum=<0 0 1>)
    Use Rotation Vector=true
   ) {module=InitialMeshOrientation001}
   /SetVariables(Particles.VisibilityTag=$System.ColorSet) {isSetParameters=true}
  ]
  particleUpdateScript [
   &ParticleState()
   &SkeletalMeshLocation(
    Bone / Socket Index=&ModuloInt(A=&ReturnParticleIDAttributes(), B=2)
    MaskSpawnGroup=false
    Mesh Sampling Type=::NewEnumerator5
    Skeletal Mesh=DataInterface() {propertyValues {
      DefaultMesh=/Game/Athena/DrivableVehicles/Hoverboard/SK_Jackal_Athena
      ExcludeBoneName=root
      OptionalComponentClass=null
      PreviewMesh=&SK_Jackal_Athena
      SkinningMode=SkinOnTheFly
      SoftSourceActor=null
      SourceMode=Default
      UvSetIndex=0
      WholeMeshLOD=-1
      bExcludeBone=false
      bReadDeformedGeometry=true
      bRequireCurrentFrameData=true
      ComponentTags []
      FilteredBones []
      FilteredSockets [FX_FrontPlate, FX_BackPlate]
      MeshUserParameter {parameter {name=None, typeDefHandle {registeredTypeIndex=1}, varData []}}
      SamplingRegions []
     }}
    Socket Sampling Mode=::NewEnumerator5
   )
   &UpdateMeshOrientation(Rotation Vector=&RandomRangeVector(Minimum=<-0.5 0 0>, Maximum=<0.5 0 0>))
   &DynamicMaterialParameters(
    Index 0 Param 1=3
    Index 0 Param 2=1
    Index 0 Param 3=1
    Index 0 Param 4=1
    Module.Index0Param0Write=true
    Module.Index0Param1Write=true
    Module.Index0Param2Write=true
    Module.Index0Param3Write=true
   )
   &ScaleColor(
    Scale Alpha=&FloatFromCurve(
     FloatCurve=DataInterface() {propertyValues {Curve {keys [<at 0 0 RCIM_Cubic>, <at 0.2 1>, {arriveTangent=-1.25, interpMode=RCIM_Cubic, leaveTangent=-1.25, time=0.8, value=1}, <at 1 0>]}}}
     Scale Curve=0.8
    )
    Scale RGB=&VectorFromFloat(Value=1.1)
    ScaleA=true
    ScaleRGB=true
   ) {module=ScaleColor001}
  ]
  rendererClasses [&NiagaraMeshRendererProperties]
  renderers [
   {rendererClass=&NiagaraMeshRendererProperties, rendererIndex=0}
   {rendererClass=&NiagaraMeshRendererProperties, rendererIndex=1}
  ]
 }
 WaterSkid {
  bEnabled=false
  simTarget=GPUComputeSim
  emitterUpdateScript [
   &EmitterState(Life Cycle Mode=::NewEnumerator1, Loop Duration=2, UseLoopDelay=true)
   &SpawnRate(
    SpawnRate=&Lerp_Float(A=0, B=600, Alpha=&Multiply_Float(A=$User.Skid, B=$User.Water))
    Use Spawn Probability=false
   )
  ]
  particleSpawnScript [
   &InitializeParticle(
    Color=&RandomRangeLinearColor(Minimum=<rgba 0.9271 0.9789 1 0.75>, Maximum=<rgba 0.9271 0.9789 1 1.25>)
    Lifetime=&RandomRangeFloat(Minimum=0.1, Maximum=0.75)
    Position Offset:NiagaraPosition=$User.Front
    Sprite Size=&RandomRangeVector2D(Minimum=<2 4>, Maximum=<5 16>)
    Sprite Size Mode=::NewEnumerator4
    UsePositionOffset=true
    Write Material Random=true
   )
   &ShapeLocation(
    Cylinder Height=60
    Cylinder Radius=5
    MaskSpawnGroup=false
    Offset=<10 0 -15>
    Offset Mode=::NewEnumerator0
    Override Seed=false
    Rotation Mode=::NewEnumerator1
    Shape Primitive=::NewEnumerator1
    Surface Only=true
    Use Endcaps In Surface Only Mode=false
    UseOwnerScale=true
   )
   &AddVelocity(Velocity=&Multiply_VectorByFloat(Vector=&Multiply_Vector(A=$Particles.ShapeLocation.ShapeVector, B=<0.8 1 1>), Float=25)) {module=AddVelocity001}
   &InheritVelocity(Inherited Velocity Amount Scale=<2 2 1>, Inherited Velocity Speed Limit=0, Speed Limit=false, UseSpeedThreshold=false)
   &AddVelocity(Velocity=&RandomRangeVector(Minimum=<0 0 300>, Maximum=<0 0 600>), Velocity Speed Scale=1.5)
   &AddVelocity(Velocity=&RandomRangeVector(Minimum=<-750 0 25>, Maximum=<-900 0 0>)) {module=AddVelocity002}
  ]
  particleUpdateScript [
   &ParticleState()
   &ScaleSpriteSize(
    Scale Factor=&Vector2DFromCurve(
     Vector2Curve=DataInterface() {propertyValues {
       XCurve {keys [
         {arriveTangent=5.7886, interpMode=RCIM_Cubic, time=0, value=0.2}
         {arriveTangent=0.8, interpMode=RCIM_Cubic, leaveTangent=0.8, time=0.2, value=0.75}
         <at 1 1 RCIM_Cubic>
        ]}
       YCurve {keys [
         {arriveTangent=9.9545, interpMode=RCIM_Cubic, time=0, value=0}
         {arriveTangent=0.3, interpMode=RCIM_Cubic, leaveTangent=0.3, time=0.2, value=0.9996}
         <at 1 0.3 RCIM_Cubic>
        ]}
      }}
     ScaleCurve=<1 1>
    )
    Scale Sprite Size Mode=::NewEnumerator2
   )
   &AccelerationForce(Acceleration=<0 0 -2200>)
   &SolveForcesAndVelocity(Clamp Velocity=false, Limit Acceleration=false)
   &ScaleSpriteSizeBySpeed(
    Max Scale Factor=<2 4>
    Min Scale Factor=<1 1.01>
    Source Velocity:Vector3f=&Multiply_Vector(A=<0 0.01 1>, B=$Particles.Velocity)
   )
   &ScaleColor(
    Scale Alpha=&FloatFromCurve(FloatCurve=DataInterface() {propertyValues {Curve {keys [<at 0 0 RCIM_Cubic>,<at 0.2 1 RCIM_Cubic>,<at 1 0 RCIM_Cubic>]}}})
    ScaleA=true
    ScaleRGB=true
   ) {module=ScaleColor001}
  ]
  rendererClasses [&NiagaraSpriteRendererProperties]
  renderers [{rendererClass=&NiagaraSpriteRendererProperties}]
 }
}
summary {
 systemName=NS_Hoverboard_Athena_SpeedFX_Converted
 emitters [
  {bEnabled=false, emitterName=SnowSkid, simTarget=GPUComputeSim, rendererClasses [&NiagaraSpriteRendererProperties]}
  {bEnabled=false, emitterName=WaterSkid, simTarget=GPUComputeSim, rendererClasses [&NiagaraSpriteRendererProperties]}
  {bEnabled=false, emitterName=SmokeSkid, rendererClasses [&NiagaraSpriteRendererProperties]}
  {bEnabled=false, emitterName=SharpSkid, rendererClasses [&NiagaraMeshRendererProperties]}
  {emitterName=RipplesFront, rendererClasses [&NiagaraSpriteRendererProperties]}
  {emitterName=UndersideGlows, rendererClasses [&NiagaraMeshRendererProperties]}
  {emitterName=LightningArcs, rendererClasses [&NiagaraMeshRendererProperties]}
  {emitterName=SpeedLinesCamera, rendererClasses [&NiagaraSpriteRendererProperties]}
  {bEnabled=false, emitterName=SPlashSkid, rendererClasses [&NiagaraMeshRendererProperties]}
  {emitterName=RipplesBack, rendererClasses [&NiagaraSpriteRendererProperties]}
 ]
 userVariables [
  User.Back {type=Vector3f, description="", defaultValue {struct=Vector3f, value=<0 0 0>}}
  User.Fast {type=Vector3f, description="", defaultValue {struct=Vector3f, value=<0 0 0>}}
  User.Front {type=NiagaraPosition, description="", defaultValue {}}
  User.HBoardColorSet {type=NiagaraInt32}
  User.Skid {type=NiagaraFloat, defaultValue=0, description=""}
  User.Snow {type=NiagaraFloat, defaultValue=0, description=""}
  User.Water {type=NiagaraFloat, defaultValue=0, description=""}
 ]
}
