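# Niagara system for the Driftboard hoverboard's water effects: mesh foam, side splashes and motor burble, GPU skid spray and sprite speed lines.
# User.InWater gates the water emitters, with spawn rates scaling by ForwardSpeed, or owner velocity for foam.
# Speed lines ignore water state and ramp steeply with speed.
# Foam and burble snap to the WaterBodyActor surface, while splashes reuse Meatball boat modules, mirroring along the board and responding to roll.
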NiagaraSystem /HoverboardV2/Vehicle/FX/NS_Vehicle_Driftboard_Water
dependencies {
 usedDataInterfaces [
  &NiagaraDataInterfaceCurve
  &NiagaraDataInterfaceVector2DCurve
  &NiagaraDataInterfaceVectorCurve
 ]
 usedDynamicInputs [&ConvertVectorToPosition, /Game/Effects/Niagara/DynamicInputs/Math/FN_RemapDotProduct]
 usedModules [
  &self:SystemSpawnScript.SystemScriptSource.SystemScriptGraph.NiagaraNodeAssignment_0.SetVariables_#1
  &SystemState
 ]
}
emitters {
 FN_Sprite_SpeedLines {
  emitterUpdateScript [
   &EmitterState(Life Cycle Mode=::NewEnumerator1, Loop Duration=0.25, UseLoopDelay=false) {module=EmitterState001}
   &SpawnBurst_Instantaneous(Spawn Count=5, Use Spawn Probability=false) {enabled=false}
   &EmitterState()
   &SpawnRate(
    SpawnRate=&Lerp_Float(A=0, B=&ClampFloat(Float=180, Max=350), Alpha=&Power(a=$User.ForwardSpeed, b=6))
    Use Spawn Probability=false
   )
  ]
  particleSpawnScript [
   /Niagara/Modules/Spawn/Initialization/InitializeParticle(
    Color=&MakeLinearColor(
     R:NiagaraFloat=0.2292
     G:NiagaraFloat=0.1989
     B:NiagaraFloat=0.1838
     A:NiagaraFloat=&UniformRangedFloat(Minimum=0.2, Maximum=0.4)
    )
    Lifetime=&UniformRangedFloat(Minimum=0.3, Maximum=0.4)
    Module.Write Color=true
    Module.Write Lifetime=true
    Module.Write Mass=true
    Module.Write Material Random=true
    Module.Write Position=false
    Module.Write Scale=false
    Module.Write SpriteRotation=true
    Module.Write SpriteSize=true
    Position=&SimulationPosition()
    Sprite Rotation=&UniformRangedFloat(Minimum=0, Maximum=360)
    Sprite Size=&UniformRangedVector2D(Minimum=<1.25 25>, Maximum=<0.8 60>)
   )
   &TorusLocation(
    Align Torus With Axis=true
    Handle Radius=32
    Large Radius=200
    MaskSpawnGroup=false
    Module.Override Randomness=false
    Offset=&Multiply_VectorByFloat(Vector=$Engine.Owner.SystemXAxis, Float=40)
    Torus Axis=<1 0 0>
   )
   &AddVelocity(Velocity=&MakeVector(X=&UniformRangedFloat(Minimum=-50, Maximum=-100), Y=0, Z=0))
   &InheritVelocity(Module.Limit Speed=true, Velocity Scale=<-0.9 -0.9 -0.9>)
  ]
  particleUpdateScript [
   &ParticleState()
   &Color(Scale Alpha=&FloatFromCurve(FloatCurve=DataInterface() {propertyValues {Curve {keys [<at 0 0 RCIM_Cubic>,<at 0.5 1 RCIM_Cubic>,<at 1 0 RCIM_Cubic>]}}}))
   &SolveForcesAndVelocity(Clamp Velocity=false, Limit Acceleration=false)
   &ScaleSpriteSizeBySpeed(Max Scale Factor=<5 10>, Min Scale Factor=<0 0.1>)
   &ScaleSpriteSize(Scale Factor=&Vector2DFromCurve(
     Vector2Curve=DataInterface() {propertyValues {
       XCurve {keys [<at -0.0069 0.5 RCIM_Cubic>,<at 0.4918 1 RCIM_Cubic>,<at 1 0.5 RCIM_Cubic>]}
       YCurve {keys [<at 0 0.0245 RCIM_Cubic>, {arriveTangent=-0.0245, interpMode=RCIM_Cubic, leaveTangent=-0.0245, time=0.4975, value=1}, <at 1 0 RCIM_Cubic>]}
      }}
     ScaleCurve=<1 1>
    ))
   &CameraOffset(Camera Offset Amount=-100)
  ]
  rendererClasses [&NiagaraSpriteRendererProperties]
  renderers [{rendererClass=&NiagaraSpriteRendererProperties}]
 }
 FN_Water_Foam_SnapToSurface_NoSkelRef {
  emitterUpdateScript [
   &EmitterState(Life Cycle Mode=::NewEnumerator1, Loop Duration=0.25, UseLoopDelay=false) {module=EmitterState001}
   &EmitterState() {enabled=false}
   &SpawnRate(
    SpawnRate=&MakeCustomFloatFromBool(
     Bool=&Boolean_AndOperation(
      Bool1=$User.InWater
      Bool2=/Game/Athena/DrivableVehicles/Meatball/Meatball_Large/FX/DynamicInputs/FN_SpawnFromPontoon(PontoonLocation:Vector3f=$User.MidPontoonBig.Location, Threshold:NiagaraFloat=100)
     )
     True Float=&Lerp_Float(A=1, B=7, Alpha=$User.ForwardSpeed)
     False Float=0
    )
    Use Spawn Probability=false
   ) {enabled=false}
   &SpawnRate(
    SpawnRate=&Multiply_Float(
     A=20
     B=&VectorLength(Vector=&NormalizeVector(Vector To Normalize=$Engine.Owner.Velocity, Safe Normalize=true, Safe Fallback Threshold=0, Safe Fallback Vector=<0 0 0.001>))
    )
    Use Spawn Probability=false
   ) {enabled=false, module=SpawnRate001}
   &SpawnRate(
    SpawnRate=&MakeCustomFloatFromBool(
     Bool=&Boolean_AndOperation(Bool1=$User.InWater, Bool2=true)
     True Float=&Lerp_Float(
      A=1
      B=30
      Alpha=&FloatFromCurve(
       FloatCurve=DataInterface() {propertyValues {Curve {keys [<at 50 0 RCIM_Cubic>, {arriveTangent=0.0423, interpMode=RCIM_Cubic, leaveTangent=0.0423, tangentMode=RCTM_User, time=101, value=1}]}}}
       CurveIndex=&VectorLength(Vector=$Engine.Owner.Velocity)
      )
     )
     False Float=0
    )
    Use Spawn Probability=false
   ) {module=SpawnRate002}
  ]
  particleSpawnScript [
   /Niagara/Modules/Spawn/Initialization/V2/InitializeParticle(
    Lifetime=&RandomRangeFloat(Minimum=2, Maximum=3)
    Mass Max=1.2
    Mass Min=0.6
    Mass Mode=::NewEnumerator3
    Mesh Scale Mode=::NewEnumerator3
    Mesh Uniform Scale=100
    Sprite Rotation Mode=::NewEnumerator2
    UsePositionOffset=false
    Write Material Random=true
   )
   &InitialMeshOrientation(
    Mesh Orientation Mode=::NewEnumerator2
    Rotation=&RandomRangeVector(Minimum=<0 0 0>, Maximum=<0 0 1>)
    Use Rotation Vector=true
   )
   /SetVariables(Particles.MeshOrientation=/Game/Athena/DrivableVehicles/Meatball/Meatball_Large/FX/DynamicInputs/FN_OwnerRotationRollandPitchRemoved()) {isSetParameters=true}
   &InheritVelocity(Module.Limit Speed=true, Velocity Scale=<-1 -1 -1>)
   &CameraOffset(Camera Offset Amount=15)
  ]
  particleUpdateScript [
   &ParticleState()
   /Game/Effects/Niagara/Modules/Water/WaterBodyLocation(
    Snap Position to Water Position=true
    Use Water Plane Position for Water Position=true
    WaterBodyActor:NiagaraDataInterfaceWater=$User.WaterBodyActor
   )
   &Drag(Drag=2, Use Linear Drag=true, Use Rotational Drag=false)
   &SolveForcesAndVelocity(Clamp Velocity=false, Limit Acceleration=false)
   &DynamicMaterialParameters(
    Index 0 Param 1=&FloatFromCurve(FloatCurve=DataInterface() {propertyValues {Curve {keys [<at -0.0046 1>,<at 1.0011 4>]}}})
    Index 0 Param 2=&FloatFromCurve(FloatCurve=DataInterface() {propertyValues {Curve {keys [<at -0.0046 0>,<at 1.0011 2>]}}})
    Index 0 Param 3=&FloatFromCurve(FloatCurve=DataInterface() {propertyValues {Curve {keys [<at -0.0046 1>,<at 1.0011 0.5>]}}})
    Index 1 Param 3=&Multiply_Float(
     A=200
     B=&Multiply_Float(A=&MakeFloatFromVector(Vector=$Particles.Scale, Channel=::NewEnumerator0), B=0.0067)
    )
    Index 2 Param 1=&FloatFromCurve(FloatCurve=DataInterface() {propertyValues {Curve {keys [<at -0.0046 0.4>,<at 1.0011 1.5>]}}})
    Index 2 Param 2=&FloatFromCurve(FloatCurve=DataInterface() {propertyValues {Curve {keys [<at -0.0046 0.1>,<at 0.11 2.6>,<at 1.0011 0.3>]}}})
    Index 2 Param 3=0.35
    Index 2 Param 4=&FloatFromCurve(FloatCurve=DataInterface() {propertyValues {Curve {keys [
         {arriveTangent=-1.9991, interpMode=RCIM_Cubic, time=0, value=1}
         {arriveTangent=-2.7143, interpMode=RCIM_Cubic, leaveTangent=-2.7143, tangentMode=RCTM_User, time=0.2095, value=0.6901}
         <at 1 0 RCIM_Cubic>
        ]}}})
    Module.Index0Param0Write=true
    Module.Index0Param1Write=true
    Module.Index0Param2Write=true
    Module.Index0Param3Write=false
    Module.Index1Param0Write=false
    Module.Index1Param1Write=false
    Module.Index1Param2Write=true
    Module.Index1Param3Write=false
    Module.Index2Param0Write=true
    Module.Index2Param1Write=true
    Module.Index2Param2Write=true
    Module.Index2Param3Write=true
    Write Parameter Index 1=true
    Write Parameter Index 2=true
   )
   &ScaleColor(
    Scale Alpha=&FloatFromCurve(FloatCurve=DataInterface() {propertyValues {Curve {keys [
         {arriveTangent=37.6118, interpMode=RCIM_Cubic, leaveTangent=37.6118, tangentMode=RCTM_User, time=0, value=0}
         {arriveTangent=0.002, interpMode=RCIM_Cubic, leaveTangent=0.002, time=0.0961, value=0.9983}
         <at 1.0011 0.002 RCIM_Cubic>
        ]}}})
    ScaleA=true
    ScaleRGB=true
   )
  ]
  rendererClasses [&NiagaraMeshRendererProperties]
  renderers [{rendererClass=&NiagaraMeshRendererProperties, rendererIndex=0}]
 }
 Mesh_Splashes_LS {
  emitterUpdateScript [
   &EmitterState(Life Cycle Mode=::NewEnumerator1, Loop Duration=0.25, UseLoopDelay=false)
   &EmitterScalability(
    Module.bApplySpawnCountScale=false
    Module.bApplySpawnCountScaleByDistanceFraction=false
    Module.bUseMaxDistance=false
    Module.bUseMinDistance=false
   ) {enabled=false}
   &EmitterLifeCycle(NextLoopDuration=0.25) {enabled=false}
   &SpawnRate(
    SpawnRate=&MakeCustomFloatFromBool(Bool=$User.InWater, True Float=&Lerp_Float(A=0, B=160, Alpha=$User.ForwardSpeed), False Float=0)
    Use Spawn Probability=false
   )
  ]
  particleSpawnScript [
   /Game/Athena/DrivableVehicles/Meatball/Meatball_Large/FX/Modules/FN_BoatInitialize(Lifetime:NiagaraFloat=&Lerp_Float(
     A=&UniformRangedFloat(Minimum=0.254, Maximum=0.55)
     B=&UniformRangedFloat(Minimum=0.325, Maximum=0.675)
     Alpha=$User.ForwardSpeed
    ))
   /Game/Athena/DrivableVehicles/Meatball/Meatball_Large/FX/Modules/V2/FN_LinePosition(
    End=75
    Offset Direction:Vector3f=&MakeVector(
     X=0
     Y=&Lerp_Float(
      A=&FloatFromCurve(
       FloatCurve=DataInterface() {propertyValues {Curve {keys [{arriveTangent=-1.5223, time=0, value=0.15}, <at 1 0.05>]}}}
       CurveIndex=$Particles.RandomDistribution
       Scale Curve=150
      )
      B=100
      Alpha=&Multiply_Float(
       A=/Niagara/DynamicInputs/Math/Min_Float(
        A:NiagaraFloat=&Abs_Float(Float:NiagaraFloat=&Multiply_Float(A=$System.BoatRoll, B=3))
        B:NiagaraFloat=1
       )
       B=&Max_Float(
        A:NiagaraFloat=&Multiply_Float(A=$Particles.BoatSign, B=&self:signScrach(float:NiagaraFloat=$System.BoatRoll))
        B:NiagaraFloat=1
       )
      )
     )
     Z=0
    )
    Start=-50
    Symmetry=true
   )
   &BoxLocation(Box Offset=<0 0 -10>, Box Size=<0 0 0>, MaskSpawnGroup=false, Module.Override Seed=false, Rotate Box=false, Surface Only=false)
   /Game/Athena/DrivableVehicles/Meatball/Meatball_Large/FX/Modules/FN_VehicleScale(
    Boost Multiplier=0.125
    Mesh Scale:Vector3f=&Multiply_VectorByFloat(Vector=<1 2 1>, Float=0.5)
    Modulate by Pontoon Depth=false
    Modulate by Pontoon Impulse=false
    Modulate by Random Float=true
    Modulate by Vehicle Speed=true
    Module.Write Mesh Scale=true
    Module.Write Sprite Size=false
    Remap Forward Speed=DataInterface() {propertyValues {Curve {keys [{arriveTangent=3.1047, interpMode=RCIM_Cubic, leaveTangent=3.1047, tangentMode=RCTM_User, time=0.121, value=-0.0129}, <at 1.0045 1.0505>]}}}
    SizeCurve=DataInterface() {propertyValues {Curve {keys [
        {arriveTangent=-0.6878, interpMode=RCIM_Cubic, leaveTangent=-0.6878, tangentMode=RCTM_User, time=0, value=1}
        {arriveTangent=-0.7009, interpMode=RCIM_Cubic, leaveTangent=-0.7009, tangentMode=RCTM_User, time=0.1862, value=0.615}
        {arriveTangent=-0.5072, interpMode=RCIM_Cubic, leaveTangent=-0.5072, tangentMode=RCTM_User, time=1, value=0.25}
       ]}}}
    Sprite Size:Vector2f=<100 250>
   )
   /Game/Athena/DrivableVehicles/Meatball/Meatball_Large/FX/Modules/FN_VehicleDirectionVector(
    Rotate on X Axis=true
    Rotate on Y Axis=true
    Rotate on Z Axis=true
    Set MeshOrientation=true
    Set Sprite Alignment=false
    X Rotation End:NiagaraFloat=&UniformRangedFloat(Minimum=0, Maximum=0)
    Y Rotation End=-0.0125
    Y Rotation Start=0.005
    Z Rotation End:NiagaraFloat=&Lerp_Float(
     A=&UniformRangedFloat(Minimum=-0.35, Maximum=-0.55)
     B=&UniformRangedFloat(Minimum=0.35, Maximum=0.55)
     Alpha=$Particles.BoatSide
    )
    Z Rotation Start:NiagaraFloat=&Lerp_Float(A=-0.35, B=0.35, Alpha=$Particles.BoatSide)
   )
  ]
  particleUpdateScript [
   &ParticleState()
   /Game/Athena/DrivableVehicles/Meatball/Meatball_Large/FX/Modules/FN_BoatRemoveOwnerRoll_v2(
    Angle:NiagaraFloat=&FloatFromCurve(
     FloatCurve=DataInterface() {propertyValues {Curve {keys [<at -0.0058 0 RCIM_Cubic>, {arriveTangent=2.6901, interpMode=RCIM_Cubic, leaveTangent=2.6901, tangentMode=RCTM_User, time=1.0011, value=1}]}}}
     Scale Curve=0
    )
    Axis:ENiagaraOrientationAxis=::YAxis
    Offset:NiagaraFloat=&FloatFromCurve(FloatCurve=DataInterface() {propertyValues {Curve {keys [<at -0.0077 0.0713>,<at 0.9942 1.5508>]}}}, Scale Curve=-20)
    Offset position along X=true
    Remove Owner Pitch and Roll from Mesh Orientation=true
    Remove Owner Pitch and Roll from Sprite Alignment=true
   )
   /Game/Athena/DrivableVehicles/Meatball/Meatball_Large/FX/Modules/FN_ScaleByVehicleData(Boost Multiplier=0.2, Inside Roll Range Max=0.3, Outside Bank Multiplier=1.5)
   &SolveForcesAndVelocity(Clamp Velocity=false, Limit Acceleration=false)
   &DynamicMaterialParameters(
    Index 0 Param 1=&FloatFromCurve(FloatCurve=DataInterface() {propertyValues {Curve {keys [
         {arriveTangent=1.2629, interpMode=RCIM_Cubic, leaveTangent=1.2629, tangentMode=RCTM_User, time=-0.0045, value=-1.9799}
         {arriveTangent=4.4591, interpMode=RCIM_Cubic, leaveTangent=4.4592, tangentMode=RCTM_User, time=1, value=0}
        ]}}})
    Index 0 Param 2=&FloatFromCurve(FloatCurve=DataInterface() {propertyValues {Curve {keys [<at 0.0102 4>,<at 0.9749 0>]}}})
    Index 0 Param 3=&MakeFloatFromVector(Vector=$Particles.Scale, Channel=::NewEnumerator0)
    Index 0 Param 4=1
    Index 1 Param 1=&FloatFromCurve(FloatCurve=DataInterface() {propertyValues {Curve {keys [<at 0.7 0 RCIM_Cubic>, {arriveTangent=6.4641, interpMode=RCIM_Cubic, leaveTangent=6.4641, tangentMode=RCTM_User, time=1, value=1}]}}})
    Module.Param0WriteEnabled=true
    Module.Param0WriteEnabled1=true
    Module.Param1WriteEnabled=true
    Module.Param1WriteEnabled1=true
    Module.Param2WriteEnabled=true
    Module.Param2WriteEnabled1=true
    Module.Param3WriteEnabled=true
    Module.Param3WriteEnabled1=true
    Write Parameter Index 1=true
   )
  ]
  rendererClasses [&NiagaraMeshRendererProperties]
  renderers [
   {rendererClass=&NiagaraMeshRendererProperties, rendererIndex=0}
   {rendererClass=&NiagaraMeshRendererProperties, rendererIndex=1}
  ]
 }
 Motor_Burble_Mesh_WS_NoSkelRef {
  emitterUpdateScript [
   &EmitterState(Life Cycle Mode=::NewEnumerator1, Loop Duration=0.25, UseLoopDelay=false)
   &SpawnRate(
    SpawnRate=&MakeCustomFloatFromBool(Bool=$User.InWater, True Float=&Lerp_Float(A=0, B=20, Alpha=$User.ForwardSpeed), False Float=0)
    Use Spawn Probability=false
   )
  ]
  particleSpawnScript [
   &InitializeParticle(
    Lifetime=&Lerp_Float(A=1.8, B=&UniformRangedFloat(Minimum=0.75, Maximum=1), Alpha=$User.ForwardSpeed)
    Mass Mode=::NewEnumerator1
    Mesh Scale=&Multiply_VectorByFloat(Vector=&UniformRangedVector(Minimum=<1.25 0.8 0.8>, Maximum=<2 1.4 1.15>), Float=0.35)
    Mesh Scale Mode=::NewEnumerator4
    Position Offset=&TransformVector(Vector=<-50 0 0>, Destination Space=::Simulation)
    Sprite Size Mode=::NewEnumerator4
    UsePositionOffset=true
    Write Material Random=true
   ) {module=InitializeParticle001}
   &WaterBodyLocation(
    Snap Position to Water Position=true
    Use Water Plane Position for Water Position=true
    WaterBodyActor:NiagaraDataInterfaceWater=$User.WaterBodyActor
   )
   /SetVariables(
    Particles.MeshOrientation=&FN_OwnerRotationRollandPitchRemoved()
    Particles.Scale=&Multiply_Vector(A=$Particles.Scale, B=&Lerp_Vector(A=<0.5 0.5 1>, B=<1 1 1>, Alpha=$User.ForwardSpeed))
   ) {isSetParameters=true}
  ]
  particleUpdateScript [
   &ParticleState()
   /SetVariables(Particles.Scale=&Multiply_Vector(
     A=$Particles.Initial.Scale
     B=&VectorFromCurve(
      VectorCurve=DataInterface() {propertyValues {XCurve {keys [<at 0 1>,<at 0.9989 1>]}, YCurve {keys [<at 0 1>,<at 0.9989 1>]}, ZCurve {keys [<at 0 0>,<at 0.2345 0.7983>,<at 0.9989 1>]}}}
      CurveIndex=$Emitter.Scalability.DistanceFraction
     )
    )) {isSetParameters=true}
   &Drag(Drag=2, Use Linear Drag=true, Use Rotational Drag=false)
   &ScaleColor(
    Scale Alpha=&FloatFromCurve(FloatCurve=DataInterface() {propertyValues {Curve {keys [
         {arriveTangent=9.9926, interpMode=RCIM_Cubic, leaveTangent=9.9927, tangentMode=RCTM_User, time=0, value=0}
         {arriveTangent=-2.8339, interpMode=RCIM_Cubic, leaveTangent=-2.8339, tangentMode=RCTM_User, time=0.0216, value=1}
         <at 1.0011 0.002 RCIM_Cubic>
        ]}}})
    ScaleA=true
    ScaleRGB=true
   )
   &InheritSourceMovement(
    Applied Position Scale=&VectorFromCurve(VectorCurve=DataInterface() {propertyValues {
       XCurve {keys [{arriveTangent=-2.6869, interpMode=RCIM_Cubic, leaveTangent=-2.6869, tangentMode=RCTM_User, time=0, value=1}, <at 1 0>]}
       YCurve {keys [{arriveTangent=-2.6869, interpMode=RCIM_Cubic, leaveTangent=-2.6869, tangentMode=RCTM_User, time=0, value=1}, <at 1 0>]}
       ZCurve {keys [<at 0 0>,<at 1 0>]}
      }})
    ApplyForce=false
    ApplyPosition=true
    ApplyPositionFalloff=false
   )
   &SolveForcesAndVelocity(Clamp Velocity=false, Limit Acceleration=false)
   /SetVariables(Particles.Scale=&Multiply_Vector(
     A=$Particles.Initial.Scale
     B=&VectorFromCurve(VectorCurve=DataInterface() {propertyValues {
        XCurve {keys [
          {arriveTangent=4.6994, interpMode=RCIM_Cubic, leaveTangent=4.6994, tangentMode=RCTM_User, time=0, value=0.0795}
          {arriveTangent=0.4589, interpMode=RCIM_Cubic, leaveTangent=0.4589, tangentMode=RCTM_User, time=0.1447, value=0.7609}
          <at 1 1>
         ]}
        YCurve {keys [{arriveTangent=2.5595, interpMode=RCIM_Cubic, leaveTangent=2.5595, tangentMode=RCTM_User, time=0, value=0.0767}, <at 1 1>]}
        ZCurve {keys [
          <at 0 0 RCIM_Cubic>
          {arriveTangent=-2.3619, interpMode=RCIM_Cubic, leaveTangent=-2.3619, tangentMode=RCTM_User, time=0.336, value=0.9567}
          <at 0.9989 0>
         ]}
       }})
    )) {isSetParameters=true}
   &DynamicMaterialParameters(
    Index 0 Param 1=&FloatFromCurve(FloatCurve=DataInterface() {propertyValues {Curve {keys [
         {arriveTangent=0.2615, interpMode=RCIM_Cubic, leaveTangent=0.2615, tangentMode=RCTM_User, time=0, value=-1.7984}
         {arriveTangent=3.2109, interpMode=RCIM_Cubic, leaveTangent=3.2109, tangentMode=RCTM_User, time=0.9985, value=0}
        ]}}})
    Index 0 Param 2=&Add_Float(
     A=&FloatFromCurve(FloatCurve=DataInterface() {propertyValues {Curve {keys [
          {arriveTangent=-9.4845, interpMode=RCIM_Cubic, leaveTangent=-9.4845, tangentMode=RCTM_User, time=0, value=4}
          {arriveTangent=0.2383, interpMode=RCIM_Cubic, leaveTangent=0.2383, tangentMode=RCTM_User, time=1, value=0}
         ]}}})
     B=0
    )
    Index 1 Param 1=&FloatFromCurve(FloatCurve=DataInterface() {propertyValues {Curve {keys [
         <at 0 0 RCIM_Cubic>
         {arriveTangent=1, interpMode=RCIM_Cubic, leaveTangent=1, time=0.5861, value=0.1787}
         {arriveTangent=3.3294, interpMode=RCIM_Cubic, leaveTangent=3.3294, tangentMode=RCTM_User, time=1, value=1}
        ]}}})
    Module.Param0WriteEnabled=true
    Module.Param0WriteEnabled1=true
    Module.Param1WriteEnabled=true
    Module.Param1WriteEnabled1=false
    Module.Param2WriteEnabled=true
    Module.Param2WriteEnabled1=false
    Module.Param3WriteEnabled=true
    Module.Param3WriteEnabled1=false
    Write Parameter Index 1=true
   )
  ]
  rendererClasses [&NiagaraMeshRendererProperties]
  renderers [{rendererClass=&NiagaraMeshRendererProperties, rendererIndex=0}]
 }
 WaterSkid {
  simTarget=GPUComputeSim
  emitterUpdateScript [
   &EmitterState(Life Cycle Mode=::NewEnumerator1, Loop Duration=2, UseLoopDelay=true)
   &SpawnRate(SpawnRate=900, Use Spawn Probability=false) {enabled=false, module=SpawnRate001}
   &SpawnRate(
    SpawnRate=&MakeCustomFloatFromBool(Bool=$User.InWater, True Float=&Lerp_Float(A=0, B=100, Alpha=$User.ForwardSpeed), False Float=0)
    Use Spawn Probability=false
   )
  ]
  particleSpawnScript [
   &InitializeParticle(
    Color=&RandomRangeLinearColor(Minimum=<rgba 0.9271 0.9789 1 0.75>, Maximum=<rgba 0.9271 0.9789 1 1.25>)
    Lifetime=&RandomRangeFloat(Minimum=0.1, Maximum=0.75)
    Position Offset:NiagaraPosition=$User.Front
    Sprite Size=&RandomRangeVector2D(Minimum=<2 4>, Maximum=<5 16>)
    Sprite Size Mode=::NewEnumerator4
    UsePositionOffset=true
    Write Material Random=true
   )
   &ShapeLocation(
    Cylinder Height=10
    Cylinder Radius=20
    MaskSpawnGroup=false
    Offset=<0 0 -25>
    Offset Mode=::NewEnumerator0
    Override Seed=false
    Rotation Axis=<0 0 1>
    Rotation Coordinate Space=::World
    Rotation Mode=::NewEnumerator1
    Shape Primitive=::NewEnumerator1
    Surface Only=true
    Use Endcaps In Surface Only Mode=false
    UseOwnerScale=true
   )
   &AddVelocity(Velocity=&Multiply_VectorByFloat(Vector=&Multiply_Vector(A=$Particles.ShapeLocation.ShapeVector, B=<0.8 1 1>), Float=25)) {enabled=false, module=AddVelocity001}
   &InheritVelocity(Inherited Velocity Amount Scale=<0 0 1>, Inherited Velocity Speed Limit=0, Speed Limit=false, UseSpeedThreshold=false)
   &AddVelocity(Velocity=&RandomRangeVector(Minimum=<0 0 400>, Maximum=<0 0 200>), Velocity Speed Scale=1.5)
   &AddVelocity(Velocity=&RandomRangeVector(Minimum=<-750 -300 0>, Maximum=<-900 300 0>)) {module=AddVelocity002}
   /SetVariables(Particles.Position=&AddVectorToPosition(Position=$Particles.Position, Vector=&TransformVector(Vector=<-35 0 0>, Destination Space=::Simulation))) {isSetParameters=true}
  ]
  particleUpdateScript [
   &ParticleState()
   &ScaleSpriteSize(
    Scale Factor=&Vector2DFromCurve(
     Vector2Curve=DataInterface() {propertyValues {
       XCurve {keys [
         {arriveTangent=5.7886, interpMode=RCIM_Cubic, time=0, value=0.2}
         {arriveTangent=0.8, interpMode=RCIM_Cubic, leaveTangent=0.8, time=0.2, value=0.75}
         <at 1 1 RCIM_Cubic>
        ]}
       YCurve {keys [
         {arriveTangent=9.9545, interpMode=RCIM_Cubic, time=0, value=0}
         {arriveTangent=0.3, interpMode=RCIM_Cubic, leaveTangent=0.3, time=0.2, value=0.9996}
         <at 1 0.3 RCIM_Cubic>
        ]}
      }}
     ScaleCurve=<1 1>
    )
    Scale Sprite Size Mode=::NewEnumerator2
   )
   &AccelerationForce(Acceleration=<0 0 -2200>)
   &InheritSourceMovement(
    Applied Position Scale=&VectorFromFloat(Value=&FloatFromCurve(FloatCurve=DataInterface() {propertyValues {Curve {keys [<at 0 1 RCIM_Cubic>, {arriveTangent=-2.0827, interpMode=RCIM_Cubic, leaveTangent=-2.0827, tangentMode=RCTM_User, time=1, value=0}]}}}))
    ApplyForce=false
    ApplyPosition=true
    ApplyPositionFalloff=false
   )
   &SolveForcesAndVelocity(Clamp Velocity=false, Limit Acceleration=false)
   &ScaleSpriteSizeBySpeed(
    Max Scale Factor=<2 4>
    Min Scale Factor=<1 1.01>
    Source Velocity:Vector3f=&Multiply_Vector(A=<0 0.01 1>, B=$Particles.Velocity)
   )
   &ScaleColor(
    Scale Alpha=&FloatFromCurve(FloatCurve=DataInterface() {propertyValues {Curve {keys [<at 0 0 RCIM_Cubic>,<at 0.2 1 RCIM_Cubic>,<at 1 0 RCIM_Cubic>]}}})
    ScaleA=true
    ScaleRGB=true
   ) {module=ScaleColor001}
  ]
  rendererClasses [&NiagaraSpriteRendererProperties]
  renderers [{rendererClass=&NiagaraSpriteRendererProperties}]
 }
}
summary {
 systemName=NS_Vehicle_Driftboard_Water
 emitters [
  {emitterName=FN_Sprite_SpeedLines, rendererClasses [&NiagaraSpriteRendererProperties]}
  {emitterName=FN_Water_Foam_SnapToSurface_NoSkelRef, rendererClasses [&NiagaraMeshRendererProperties]}
  {emitterName=Mesh_Splashes_LS, rendererClasses [&NiagaraMeshRendererProperties]}
  {emitterName=Motor_Burble_Mesh_WS_NoSkelRef, rendererClasses [&NiagaraMeshRendererProperties]}
  {emitterName=WaterSkid, simTarget=GPUComputeSim, rendererClasses [&NiagaraSpriteRendererProperties]}
 ]
 userVariables [
  User.Boost {type=NiagaraFloat, defaultValue=0, description=""}
  User.DebugWaterIndex {type=NiagaraInt32}
  User.DefaultWakeStrength {type=NiagaraFloat, defaultValue=0, description=""}
  User.EnableSpawn {type=NiagaraBool}
  User.ForwardSpeed {type=NiagaraFloat, defaultValue=1, description=""}
  User.ForwardVelocity {type=NiagaraFloat, defaultValue=1, description=""}
  User.Front {type=NiagaraPosition, description="", defaultValue {}}
  User.HasDriver {type=NiagaraBool}
  User.ImmersionDepth {type=NiagaraFloat, defaultValue=0, description=""}
  User.Impulse {type=Vector3f, description="", defaultValue {struct=Vector3f, value=<0 0 0>}}
  User.InAir {type=NiagaraBool}
  User.InWater {type=NiagaraBool, defaultValue=true}
  User.isBoosting {type=NiagaraBool}
  User.IsInWaterCenter {type=NiagaraFloat, defaultValue=0, description=""}
  User.NewVariable {type=NiagaraFloat, defaultValue=0, description=""}
  User.PontoonDepth {type=NiagaraFloat, defaultValue=0, description=""}
  User.PontoonLocation {type=Vector3f, description="", defaultValue {struct=Vector3f, value=<0 0 0>}}
  User.PontoonRotation {type=Quat4f, description="", defaultValue {struct=Quat4f, value=<0 0 0 1>}}
  User.RearPontoonBig.Depth {type=NiagaraFloat, defaultValue=0, description=""}
  User.RearPontoonBig.WaterIndex {type=NiagaraInt32}
  User.Skid {type=NiagaraFloat, defaultValue=0, description=""}
  User.TurnDirection {type=NiagaraFloat, defaultValue=0, description=""}
  User.TurningAngle {type=NiagaraFloat, defaultValue=0, description=""}
  User.TurningSpeed {type=NiagaraFloat, defaultValue=0, description=""}
  User.TurnSpeed {type=NiagaraFloat, defaultValue=0, description=""}
  User.WaterBodyActor {
   type=NiagaraDataInterfaceWater
   description=""
   defaultValue {
    struct=NiagaraExt_VariableValue_DataInterface
    value {
     dataInterface=&self:NiagaraDataInterfaceWater_0
     dataInterfaceClass=/Script/Water.NiagaraDataInterfaceWater
    }
   }
  }
  User.WaterHeight {type=NiagaraFloat, defaultValue=0, description=""}
 ]
}
