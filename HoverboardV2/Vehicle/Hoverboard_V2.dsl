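; Hoverboard V2 vehicle Blueprint that overrides parent cosmetic events to replace hard-set Cascade effects with Niagara versions.
; Death, jump, impact, landing, boost and water events drive effects, audio and feedback, with most spawned systems receiving the FXColorSet parameter.
; While in water, a timer updates surface height and loop audio position.
; Empty-fuel and refuel events gate boosting; cosmetic setup spawns a grass-push wash force only at high effects quality.
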(event EventBeginPlay
  (|CallParentFunction)
  (Effects|Components|ParticleSystem|SetIntParameter (Variables|Default|GetBoostFX_NS) "HBoardColorSet" (Variables|Default|GetFXColorSet))
  (EventDispatchers|BindEventtoOnRefueledfromEmptyDelegate (EventDispatchers|CreateEvent :!event Custom|HandleRefueledFromEmpty)))

(event Collision|EventActorBeginOverlap
  (params (OtherActor :type Actor%20Object%20Reference)))

(event EventTick
  (params (DeltaSeconds :type Float%20%28single-precision%29)))

(event EventOnEmptyFuel
  (Variables|Default|SetOutOfFuel true))

(event Custom|HandleRefueledFromEmpty
  (Variables|Default|SetOutOfFuel false))

(event EventOnDeathPlayEffects
  (params (Damage :type Float%20%28single-precision%29) (DamageTags :type Gameplay%20Tag%20Container%20Structure%20%28by%20ref%29) (Momentum :type Vector :default "0, 0, 0") (HitInfo :type Hit%20Result%20Structure%20%28by%20ref%29) (InstigatedBy :type Fort%20Pawn%20Object%20Reference) (DamageCauser :type Actor%20Object%20Reference) (EffectContext :type Gameplay%20Effect%20Context%20Handle%20Structure))
  (bind _returnvalue (Transformation|GetActorLocation))
  (bind (_returnvalue_location _returnvalue_rotation) (Math|Transform|BreakTransform (Transformation|GetSocketTransform (Variables|Vehicle|GetSkeletalMesh) "FrameCenter" "RTS_World")))
  ;[1 1 1 0.2] Overridden functions that were copied/reimplaneted with niagara, originals had cascade FX hard set into the functions
  ;;[1 1 1] EVENT -> On Death Play Effects
  (bind _returnvalue_2 (Niagara|SpawnSystematLocation "/HoverboardV2/Vehicle/FX/NS_Hoverboard_Athena_Destroy_Converted.NS_Hoverboard_Athena_Destroy_Converted" _returnvalue_location _returnvalue_rotation))
  (Effects|Components|ParticleSystem|SetVectorParameter _returnvalue_2 "Velocity" (Vehicle|GetVehicleVelocity))
  (Effects|Components|ParticleSystem|SetIntParameter _returnvalue_2 "HBoardColorSet" (Variables|Default|GetFXColorSet))
  (Camera|PlayWorldCameraShake "/Game/Effects/Fort_Effects/BluePrints/CameraShakes/B_Small_Vertical_Jolt_CameraShake.B_Small_Vertical_Jolt_CameraShake_C" _returnvalue 1.5e+02 6e+02)
  (if (Variables|Sounds|GetbUseAudio)
    (Audio|PlaySoundatLocation (Variables|Sounds|GetOnDeathSound) _returnvalue)
    (else
      _))
  (Development|SetHiddeninGame (Variables|Vehicle|GetSkeletalMesh) true false))

(event EventOnJumped
  (params (Charge :type Float%20%28single-precision%29))
  (bind _drivingplayercontroller (Variables|Hoverboard|GetDrivingPlayerController))
  ;[1 1 1 0.2] Overridden functions that were copied/reimplaneted with niagara, originals had cascade FX hard set into the functions
  ;;[1 1 1] EVENT -> Jump Charge Begin/End
  (Variables|Hoverboard|SetJumped true)
  (Variables|Hoverboard|SetbJumpCharging false)
  (Networking|CanExecuteCosmeticEvents
    (:True
      (bind _returnvalue (Niagara|SpawnSystemAttached "/HoverboardV2/Vehicle/FX/NS_Hoverboard_Athena_Jump_Converted.NS_Hoverboard_Athena_Jump_Converted" (Variables|Vehicle|GetSkeletalMesh) "FrameCenter" "0, 0, 0" "0, 0, 0" "KeepRelativeOffset" true))
      (Effects|Components|ParticleSystem|SetIntParameter _returnvalue "HBoardColorSet" (Variables|Default|GetFXColorSet))
      (if (Variables|Sounds|GetbUseAudio)
        (Components|Activation|Activate (Variables|Default|GetAudioJump) true)
        (else
          _))
      (Utilities|IsValid _drivingplayercontroller
        (:Is%20Valid
          (CameraShakes|StartCameraShake :self (Variables|/Script/Engine.PlayerController|GetPlayerCameraManager _drivingplayercontroller) :ShakeClass (Variables|Hoverboard|GetJumpCameraShake) :Scale 1.0 :PlaySpace "CameraLocal"))))))

(event EventOnCollisionHitEffects
  (params (HitLocation :type Vector%20%28by%20ref%29 :default "0, 0, 0") (HitNormalImpulse :type Vector%20%28by%20ref%29 :default "0, 0, 0") (HitFrictionImpulse :type Vector%20%28by%20ref%29 :default "0, 0, 0") (HitNormal :type Vector%20%28by%20ref%29 :default "0, 0, 0") (HitActor :type Actor%20Object%20Reference) (HitSurfaceType :type EPhysicalSurface%20Enum :default "SurfaceType_Default"))
  (bind (_asfort_pawn _bsuccess) (Utilities|Casting|CastToFortPawn HitActor))
  (bind _landingfximpactthreshold (Variables|FX|GetLandingFXImpactThreshold))
  ;[1 1 1 0.2] Overridden functions that were copied/reimplaneted with niagara, originals had cascade FX hard set into the functions
  ;;[1 1 1] EVENT -> Collision Hit FX/Audio
  (bind (_outresult _outmagnitude) (Audio|OnCollision (Variables|Default|GetFortCollisionAudioImpacts) HitLocation HitNormalImpulse))
  (if _outresult
    (Utilities|FlowControl|Sequence
      (:then_0
        (CallFunction|HapticsonHit _outmagnitude)
        (if (> _outmagnitude 1.5e+04)
          (bind _returnvalue (Niagara|SpawnSystematLocation (select _bsuccess "/HoverboardV2/Vehicle/FX/NS_Shopping_Cart_Vehicle_HitPlayer_Converted.NS_Shopping_Cart_Vehicle_HitPlayer_Converted" "/HoverboardV2/Vehicle/FX/NS_Shopping_Cart_Vehicle_HitDust_Converted.NS_Shopping_Cart_Vehicle_HitDust_Converted") HitLocation (Math|Conversions|RotationFromXVector HitNormal)))
          (Effects|Components|ParticleSystem|SetIntParameter _returnvalue "HBoardColorSet" (Variables|Default|GetFXColorSet))
          (Effects|Components|ParticleSystem|SetEmitterEnable _returnvalue "Sparks" (> _outmagnitude 4.5e+04)) ; Turn on sparks for heavy hits
        ))
      (:then_1
        (if (and (> _outmagnitude _landingfximpactthreshold) (> (Math|Vector|DotProduct HitNormal (Transformation|GetActorUpVector)) 0.6))
          (bind _returnvalue_2 (Niagara|SpawnSystemAttached "/HoverboardV2/Vehicle/FX/NS_Hoverboard_Athena_Land_Converted.NS_Hoverboard_Athena_Land_Converted" (Variables|Vehicle|GetSkeletalMesh) "None" "0.000000,0.000000,-20.000000" "0, 0, 0" "KeepRelativeOffset" true))
          (Effects|Components|ParticleSystem|SetIntParameter _returnvalue_2 "HBoardColorSet" (Variables|Default|GetFXColorSet))
          (Effects|Components|ParticleSystem|SetEmitterEnable _returnvalue_2 "Lightning" (> _outmagnitude (* _landingfximpactthreshold 1.25))) ; Only do electricity for heavy lands
          (CallFunction|UpdateSurfaceType HitSurfaceType)
          (if (Variables|Sounds|GetbUseAudio)
            (Audio|Components|Audio|Play (Variables|Default|GetAudioLand) 0.0)))))))

(event Vehicle|Water|EventOnEnteredWaterVolume
  (params (WaterSurfacePoint :type Vector%20%28by%20ref%29 :default "0, 0, 0"))
  (bind (_returnvalue_x _returnvalue_y _returnvalue_z) (Math|Vector|BreakVector (Transformation|GetActorLocation)))
  ;[1 1 1 0.2] Overridden functions that were copied/reimplaneted with niagara, originals had cascade FX hard set into the functions
  ;;[0.2379 0.87 1] EVENT -> Entered Water Volume
  (Variables|Default|SetWaterHeight (.z WaterSurfacePoint))
  (Networking|CanExecuteCosmeticEvents
    (:True
      (if (> (Rendering|GetRenderingDetailMode) 0)
        (bind _watereffectscomponent (Variables|FX|GetWaterEffectsComponent)
          (:then
            (Water|SetWaterBodyComponent _watereffectscomponent "WaterBodyActor" (Variables|/Script/Water.WaterBody|GetWaterBodyComponent (stale EventOnEnteredWaterBody WaterBody))))
          (:else
            _))
        (bind _returnvalue (Utilities|Time|SetTimerbyEvent (EventDispatchers|CreateEvent :!event CheckWaterEvent) 0.03 true))
        (Variables|Hoverboard|SetCheckWaterTimer _returnvalue)
        (bind _inwaterfx_ns (Variables|Default|GetinWaterFX_NS)
          (:then
            (Components|DestroyComponent _inwaterfx_ns))
          (:else
            _))
        (bind _returnvalue_2 (Niagara|SpawnSystemAttached "/HoverboardV2/Vehicle/FX/NS_Vehicle_In_Water_Athena_Converted.NS_Vehicle_In_Water_Athena_Converted" (Variables|Vehicle|GetSkeletalMesh) "None" (Math|Vector|MakeVector 0.0 0.0 (- (+ (Variables|Default|GetWaterHeight) (Variables|Default|GetWaterFXHeightOffset)) _returnvalue_z)) "0, 0, 0" "KeepRelativeOffset" true true "None" true))
        (Variables|Default|SetinWaterFX_NS _returnvalue_2)
        (Effects|Components|ParticleSystem|SetIntParameter _returnvalue_2 "HBoardColorSet" (Variables|Default|GetFXColorSet))
        (bind _returnvalue_3 (Niagara|SpawnSystematLocation "/HoverboardV2/Vehicle/FX/NS_WaterSplash_Athena_Vehicle_Converted.NS_WaterSplash_Athena_Vehicle_Converted" (- (Transformation|GetWorldLocation (Variables|Vehicle|GetSkeletalMesh)) "0.000000,0.000000,16.000000")))
        (Effects|Components|ParticleSystem|SetIntParameter _returnvalue_3 "HBoardColorSet" (Variables|Default|GetFXColorSet))
        (bind _inwaterloop (Variables|Default|GetInWaterLoop)
          (:then
            (Audio|Components|Audio|Stop _inwaterloop))
          (:else
            _))
        (Audio|SpawnSoundAttached "/Game/Athena/Sounds/Vehicles/GolfCart/GolfCart_Movement_Impact_Water_Cue.GolfCart_Movement_Impact_Water_Cue" (Variables|Vehicle|GetSkeletalMesh))
        (bind _returnvalue_4 (Audio|SpawnSoundAttached "/HoverboardV2/Vehicle/Audio/IP_LinedNotebookVehicle_InWater_Loop_Cue.IP_LinedNotebookVehicle_InWater_Loop_Cue" (Variables|Vehicle|GetSkeletalMesh)))
        (Variables|Default|SetInWaterLoop _returnvalue_4)))))

(event EventOnEnteredWaterBody
  (params (WaterBody :type Fort%20Water%20Body%20Actor%20Object%20Reference))
  (bind (_returnvalue_x _returnvalue_y _returnvalue_z) (Math|Vector|BreakVector (Transformation|GetActorLocation)))
  ;[1 1 1 0.2] Overridden functions that were copied/reimplaneted with niagara, originals had cascade FX hard set into the functions
  ;;[0.2379 0.87 1] EVENT -> Entered Water Volume
  (bind _outwatersurfaceinfo (WaterBody|GetWaterSurfaceInfoStruct WaterBody (Transformation|GetSocketLocation (Variables|Vehicle|GetSkeletalMesh) (Variables|Default|GetWater%20Test%20Socket%20Name)) true false false))
  (bind (_watersurfacelocation_x _watersurfacelocation_y _watersurfacelocation_z) (Math|Vector|BreakVector (Utilities|Struct|BreakWaterSurfaceInfo _outwatersurfaceinfo)))
  (Variables|Default|SetWaterHeight _watersurfacelocation_z)
  (Networking|CanExecuteCosmeticEvents
    (:True
      (if (> (Rendering|GetRenderingDetailMode) 0)
        (bind _watereffectscomponent (Variables|FX|GetWaterEffectsComponent)
          (:then
            (Water|SetWaterBodyComponent _watereffectscomponent "WaterBodyActor" (Variables|/Script/Water.WaterBody|GetWaterBodyComponent WaterBody)))
          (:else
            _))
        (bind _returnvalue (Utilities|Time|SetTimerbyEvent (EventDispatchers|CreateEvent :!event CheckWaterEvent) 0.03 true))
        (Variables|Hoverboard|SetCheckWaterTimer _returnvalue)
        (bind _inwaterfx_ns (Variables|Default|GetinWaterFX_NS)
          (:then
            (Components|DestroyComponent _inwaterfx_ns))
          (:else
            _))
        (bind _returnvalue_2 (Niagara|SpawnSystemAttached "/HoverboardV2/Vehicle/FX/NS_Vehicle_In_Water_Athena_Converted.NS_Vehicle_In_Water_Athena_Converted" (Variables|Vehicle|GetSkeletalMesh) "None" (Math|Vector|MakeVector 0.0 0.0 (- (+ (Variables|Default|GetWaterHeight) (Variables|Default|GetWaterFXHeightOffset)) _returnvalue_z)) "0, 0, 0" "KeepRelativeOffset" true true "None" true))
        (Variables|Default|SetinWaterFX_NS _returnvalue_2)
        (Effects|Components|ParticleSystem|SetIntParameter _returnvalue_2 "HBoardColorSet" (Variables|Default|GetFXColorSet))
        (bind _returnvalue_3 (Niagara|SpawnSystematLocation "/HoverboardV2/Vehicle/FX/NS_WaterSplash_Athena_Vehicle_Converted.NS_WaterSplash_Athena_Vehicle_Converted" (- (Transformation|GetWorldLocation (Variables|Vehicle|GetSkeletalMesh)) "0.000000,0.000000,16.000000")))
        (Effects|Components|ParticleSystem|SetIntParameter _returnvalue_3 "HBoardColorSet" (Variables|Default|GetFXColorSet))
        (bind _inwaterloop (Variables|Default|GetInWaterLoop)
          (:then
            (Audio|Components|Audio|Stop _inwaterloop))
          (:else
            _))
        (Audio|SpawnSoundAttached "/Game/Athena/Sounds/Vehicles/GolfCart/GolfCart_Movement_Impact_Water_Cue.GolfCart_Movement_Impact_Water_Cue" (Variables|Vehicle|GetSkeletalMesh))
        (bind _returnvalue_4 (Audio|SpawnSoundAttached "/HoverboardV2/Vehicle/Audio/IP_LinedNotebookVehicle_InWater_Loop_Cue.IP_LinedNotebookVehicle_InWater_Loop_Cue" (Variables|Vehicle|GetSkeletalMesh)))
        (Variables|Default|SetInWaterLoop _returnvalue_4)))))

(event EventCheckWaterEvent
  (bind (_returnvalue_x _returnvalue_y) (Math|Vector|BreakVector (Transformation|GetWorldLocation (Variables|Vehicle|GetSkeletalMesh))))
  (bind _waterfxheightoffset (Variables|Default|GetWaterFXHeightOffset))
  (bind _waterheight (Variables|Default|GetWaterHeight))
  (bind _returnvalue (+ _waterheight _waterfxheightoffset))
  ;[1 1 1 0.2] Overridden functions that were copied/reimplaneted with niagara, originals had cascade FX hard set into the functions
  ;;[0.2379 0.87 1] EVENT -> Entered Water Volume
  (Variables|Default|SetWaterHeight (.z (WaterBody|GetWaterSurfaceInfoatLocation (Variables|/Script/Water.WaterBody|GetWaterBodyComponent (stale EventOnEnteredWaterBody WaterBody)) (Transformation|GetActorLocation) true))) ; Get the latest WaterLevel from native
  (Transformation|SetWorldLocation (Variables|Default|GetInWaterLoop) (Math|Vector|MakeVector _returnvalue_x _returnvalue_y _returnvalue) false true)
  (Effects|Components|ParticleSystem|SetBoolParameter (Variables|FX|GetWaterEffectsComponent) "inWater" true)
  (Vehicle|CheckWater _returnvalue 0 (Variables|Default|GetInWaterLoop)))

(event EventOnExitedWaterBody
  (params (WaterBody :type Fort%20Water%20Body%20Actor%20Object%20Reference))
  ;[1 1 1 0.2] Overridden functions that were copied/reimplaneted with niagara, originals had cascade FX hard set into the functions
  ;;[0.2379 0.87 1] EVENT -> Entered Water Volume
  ;;;[0.2541 0.3275 1] EVENT -> Exited Water Volume
  (Networking|CanExecuteCosmeticEvents
    (:True
      (Utilities|Time|ClearandInvalidateTimerbyHandle (Variables|Hoverboard|GetCheckWaterTimer))
      (bind _inwaterfx_ns (Variables|Default|GetinWaterFX_NS)
        (:then
          (Components|Activation|SetActive _inwaterfx_ns false false)
          (bind _inwaterloop (Variables|Default|GetInWaterLoop)
            (:then
              (Audio|Components|Audio|FadeOut _inwaterloop 3.0 0.0 "Linear" :!class /Script/Engine.AudioComponent)
              (Effects|Components|ParticleSystem|SetBoolParameter (Variables|FX|GetWaterEffectsComponent) "inWater" false))
            (:else)))
        (:else)))))

(event Vehicle|Water|EventOnExitedWaterVolume
  (params (WaterSurfacePoint :type Vector%20%28by%20ref%29 :default "0, 0, 0"))
  ;[1 1 1 0.2] Overridden functions that were copied/reimplaneted with niagara, originals had cascade FX hard set into the functions
  ;;[0.2379 0.87 1] EVENT -> Entered Water Volume
  ;;;[0.2541 0.3275 1] EVENT -> Exited Water Volume
  (Networking|CanExecuteCosmeticEvents
    (:True
      (Utilities|Time|ClearandInvalidateTimerbyHandle (Variables|Hoverboard|GetCheckWaterTimer))
      (bind _inwaterfx_ns (Variables|Default|GetinWaterFX_NS)
        (:then
          (Components|Activation|SetActive _inwaterfx_ns false false)
          (bind _inwaterloop (Variables|Default|GetInWaterLoop)
            (:then
              (Audio|Components|Audio|FadeOut _inwaterloop 3.0 0.0 "Linear" :!class /Script/Engine.AudioComponent)
              (Effects|Components|ParticleSystem|SetBoolParameter (Variables|FX|GetWaterEffectsComponent) "inWater" false))
            (:else)))
        (:else)))))

(event EventOnBoostBegin
  (bind _driverpawn (Variables|Hoverboard|GetDriverPawn))
  (bind _returnvalue (Animation|GetCurrentMontage _driverpawn :!class /Script/Engine.Character))
  ;[1 1 1 0.2] Overridden functions that were copied/reimplaneted with niagara, originals had cascade FX hard set into the functions
  ;;[1 1 1] EVENT -> Boost Begin/End
  (CallFunction|SlipStreamCheck)
  (Utilities|IsValid _driverpawn
    (:Is%20Valid
      (if (Utilities|Array|ContainsItem (Variables|Default|GetMontagesToStopOnBoost) _returnvalue)
        (Animation|StopAnimMontage _driverpawn _returnvalue) ; Stop Pickaxe Swinging- Fixes FORT-146008 & FORT-379313 
        (else
          _))
      (Networking|CanExecuteCosmeticEvents
        (:True
          (Components|Activation|SetActive (Variables|Default|GetBoostFX_NS) true false)
          (Variables|FX|SetFXBoostEnd_0 (* (Transformation|GetActorForwardVector) -9e+01))
          (CallFunction|StartDriverCameraShake))))))

(event EventOnBoostEnd
  ;[1 1 1 0.2] Overridden functions that were copied/reimplaneted with niagara, originals had cascade FX hard set into the functions
  ;;[1 1 1] EVENT -> Boost Begin/End
  (Networking|CanExecuteCosmeticEvents
    (:True
      (Components|Activation|SetActive (Variables|Default|GetBoostFX_NS) false false)
      (CallFunction|StopDriverCameraShake))))

(event EventSetupCosmetics
  ;[1 1 1 0.2] Overridden functions that were copied/reimplaneted with niagara, originals had cascade FX hard set into the functions
  ;;[1 1 1] Setup Cosmetics
  (Networking|CanExecuteCosmeticEvents
    (:True
      (if (not (Variables|Default|GetCosmeticsInitialized))
        ;;;[1 1 1] Grass Push
        (Variables|Default|SetCosmeticsInitialized true)
        ;;;;[0.605 0.2497 1] Only spawn FN Wash Force with Epic or higher VFX quality settings
        (bind _returnvalue (Settings|GetGameUserSettings))
        (if (>= (Settings|GetVisualEffectQuality _returnvalue) 3)
          (bind _returnvalue_2 (Game|SpawnActorfromClass "/Game/Effects/Niagara/Blueprints/PhsyicsFields/FN_WashForce.FN_WashForce_C" (Transformation|GetWorldTransform (Variables|Vehicle|GetSkeletalMesh)) "Undefined" "MultiplyWithRoot" 0 2e+02 (Variables|Default|GetRotorWashForceStrength) -999.0 0))
          (bind _output_get (Variables|Default|SetPhysical%20Wash%20Force _returnvalue_2))
          (Transformation|AttachActorToComponent _output_get (Variables|Vehicle|GetSkeletalMesh) "None" "SnapToTarget" "KeepRelative" "KeepRelative" true)
          (else
            _))
        (Components|Activation|Activate (Variables|Default|GetNS_LoopingFX) false)
        (Rendering|SetVisibility (Variables|Default|GetFrontLight) true false)
        (Rendering|SetVisibility (Variables|Default|GetSM_Hoverboard_Lightbeams) false false)
        ;;;[0.9201 0.2258 1] Sounds
        (if (Variables|Sounds|GetbUseAudio)
          (Audio|SpawnSoundAttached (Variables|Sounds|GetMountSound) (Variables|Vehicle|GetSkeletalMesh)))))))

(fn CanBoost
  (return (not (Variables|Default|GetOutOfFuel))))

(fn DeactivateComponents
  (bind _physical_wash_force (Variables|Default|GetPhysical%20Wash%20Force))
  (|CallParentFunction)
  (Variables|Default|SetCosmeticsInitialized false)
  (Components|Activation|Deactivate (Variables|Default|GetNS_LoopingFX))
  (Utilities|IsValid _physical_wash_force
    (:Is%20Valid
      (Actor|DestroyActor _physical_wash_force))) ; Don't spawn the Rotor Wash multiple times
)

(fn ConstructionScript
  (|CallParentFunction))
